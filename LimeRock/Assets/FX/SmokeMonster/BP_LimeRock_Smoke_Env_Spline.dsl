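; Environment smoke effect Blueprint for LimeRock that pairs a surround-smoke Niagara system with two splines.
; The construction script passes Spline1 and Spline2 to the NS_LimeRock_Smoke_Surround component as the object variables SplineObj1 and SplineObj2.
; BeginPlay, overlap and Tick events are empty, so the Blueprint's only logic is this binding of splines to Niagara.
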(event EventBeginPlay)

(event Collision|EventActorBeginOverlap
  (params (OtherActor :type Actor%20Object%20Reference)))

(event EventTick
  (params (DeltaSeconds :type Float%20%28single-precision%29)))

(fn ConstructionScript
  (bind _ns_limerock_smoke_surround (Variables|Default|GetNS_LimeRock_Smoke_Surround))
  (Niagara|SetNiagaraVariable(Object) _ns_limerock_smoke_surround "SplineObj1" (Variables|Default|GetSpline1))
  (Niagara|SetNiagaraVariable(Object) _ns_limerock_smoke_surround "SplineObj2" (Variables|Default|GetSpline2)))
