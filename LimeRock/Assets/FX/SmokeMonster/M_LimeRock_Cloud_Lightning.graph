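# Unlit translucent Niagara sprite material for the LimeRock smoke monster's cloud puffs, with lightning-like internal lights.
# A spherized, panning cloud texture erodes a round particle mask, combined with a portal mask and depth fade for soft opacity.
# Dynamic parameters position two SphereMask lights, the second dropped on low quality.
# Base color brightens inside the portal, alongside heart glow and fake-normal Fresnel rim light; camera offset scales with particle size.
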Material /LimeRock/Assets/FX/SmokeMonster/M_LimeRock_Cloud_Lightning
graph {
 properties {
  bAutomaticallySetUsageInEditor=false
  bUsedWithNiagaraSprites=true
  blendMode=BLEND_Translucent
  shadingModel=MSM_Unlit
  translucencyLightingMode=TLM_VolumetricPerVertexDirectional
  usedShadingModels=Unlit
 }
 nodes [
  n1 = ComponentMask(None=ParticleSize()) {r=true}
  MaterialFunctionCall(Offset Amount=@n1, Clamp Padding=Multiply(A=@n1, B=ScalarParameter() {defaultValue=-5}) {constB=1.5}, World Position=ParticlePositionWS()) {materialFunction=/Engine/Functions/Engine_MaterialFunctions02/WorldPositionOffset/CameraOffset}
  n2 = ParticleColor()
  n3 = ConstantBiasScale(None=TextureCoordinate()) {bias=-0.5, scale=2}
  n4 = OneMinus(None=Saturate(None=DotProduct(A=@n3, B=@n3)))
  n5 = Divide(A=ConstantBiasScale(None=TextureCoordinate()) {bias=-0.5, scale=2}) {constB=1.2}
  n6 = Multiply(A=@n5)
  n7 = OneMinus(None=Saturate(None=DotProduct(A=@n6, B=@n6)))
  n8 = ScalarParameter() {defaultValue=0.7, parameterName=Spherize UV}
  n9 = Divide(A=@n5, B=Add(A=Multiply(A=@n7, B=@n8) {constB=0.5}))
  n10 = Add(A=ParticleRandom(), B=Multiply(A=ConstantBiasScale(None=@n9), B=ScalarParameter() {defaultValue=1, parameterName=CloudTile}) {constB=0.25})
  n11 = DynamicParameter() {parameterIndex=2, paramNames [PanSpeed,OpacityMult,Param3,Param4]}
  n12 = Multiply(
   A=AppendVector(A=ScalarParameter() {defaultValue=0.05, parameterName=Cloud Twist}, B=ScalarParameter() {defaultValue=-0.1, parameterName=Cloud Rise})
   B=@n11.PanSpeed
  )
  n13 = OneMinus(None=TextureSample(UVs=Panner(Coordinate=@n10, Speed=@n12) {speedY=0.1}) {samplerType=SAMPLERTYPE_LinearGrayscale, texture=/Game/Environments/Helios/TechArt/LowAltitudeClouds/T_LowAltitudeClouds_A}.G)
  n14 = ScalarParameter() {defaultValue=0.2, parameterName=NoiseSub}
  n15 = LinearInterpolate(Alpha=@n2.A) {constA=1, constB=0}
  n16 = Subtract(A=Subtract(A=SquareRoot(None=@n4), B=Multiply(A=@n13, B=@n14)), B=@n15)
  n17 = OneMinus(None=@n15)
  n18 = Power(Base=Saturate(None=Divide(A=@n16, B=@n17)))
  n19 = Multiply(A=@n2.A, B=@n18)
  n20 = NamedRerouteDeclaration(None=@n19) {name=Sphere Mask, nodeColor=<rgba 0.6406 0 1 1>}
  n21 = ScalarParameter() {defaultValue=0.7, parameterName=Base Exp}
  n22 = Multiply(
   A=VectorParameter() {parameterName=Base Color, defaultValue=<rgba 0.0384 0.032 0.1 1>}.RGB
   B=LinearInterpolate(
    A=ScalarParameter() {defaultValue=1, parameterName=Base 01}
    B=ScalarParameter() {defaultValue=0.3, parameterName=Base 02}
    Alpha=Power(Base=@n20, Exp=@n21) {constExponent=1.5}
   )
  )
  n23 = MaterialFunctionCall!(MaskEnabled) {materialFunction=/LimeRock/Assets/Materials/MF_LImeRock_PortalOpacityMask}
  n24 = LinearInterpolate(
   B=Constant3Vector() {constant=<rgba 0.923 0.1765 1 1>}
   Alpha=NamedRerouteDeclaration(None=@n23.InteriorMask) {name=InsidePortalMask, nodeColor=<rgba 1 0.75 0 1>}
  ) {constA=1, constB=5}
  n25 = NamedRerouteUsage()
  n26 = Add(A=Add(A=NamedRerouteUsage(), B=Multiply(A=@n2.RGB, B=@n25)), B=NamedRerouteUsage())
  n27 = Constant()
  MaterialFunctionCall(Emissive=Add(A=Multiply(A=@n22, B=@n24), B=StaticSwitchParameter(True=@n26, False=@n27) {defaultValue=true, parameterName=UseInternalLights})) {materialFunction=&MF_FX_EmissiveAdjustments}
  n28 = NamedRerouteUsage()
  n29 = ParticleRadius()
  n30 = Subtract(
   A=MaterialFunctionCall!(Fraction Offset, Use Custom Depth, Pixel Offset) {materialFunction=/Engine/Functions/Engine_MaterialFunctions02/SampleSceneDepth}.Depth Using Pixel Offset
   B=Multiply(A=OneMinus(None=@n28), B=Multiply(A=@n29) {constB=0.5})
  )
  n31 = PixelDepth()
  n32 = ParticleRadius()
  n33 = Saturate(None=ScalarParameter() {defaultValue=0.4, parameterName=Depth})
  n34 = Saturate(None=Divide(A=Subtract(A=@n30, B=@n31.R), B=Multiply(A=@n32, B=@n33)) {constB=500})
  n35 = Saturate(None=Divide(A=Multiply(A=@n19, B=@n34), B=ScalarParameter() {defaultValue=0.5, parameterName=Softness}))
  Multiply(A=Multiply(A=@n23.OpacityMask, B=Power(Base=@n35)), B=NamedRerouteUsage())
  n36 = NamedRerouteUsage()
  n37 = ScalarParameter() {defaultValue=0.9, parameterName=Particle Density Multiplier}
  n38 = ParticleRadius()
  NamedRerouteDeclaration(None=Multiply(A=Power(Base=@n36), B=Multiply(A=@n37, B=@n38))) {name=Particle Light Density, nodeColor=<rgba 0.875 1 0 1>}
  n39 = DynamicParameter()
  n40 = ActorPositionWS()
  n41 = WorldPosition() {worldPositionShaderOffset=WPT_ExcludeAllShaderOffsets}
  n42 = Multiply(A=VertexNormalWS(), B=NamedRerouteUsage())
  n43 = Multiply(
   A=@n39.A
   B=SphereMask(
    A=Add(A=@n40, B=@n39.RGB)
    B=Add(A=@n41.XYZ, B=@n42)
    Radius=ScalarParameter() {defaultValue=16000, parameterName=LightRadius}
    Hardness=ScalarParameter() {defaultValue=0.05, parameterName=LightHardness}
   )
  )
  n44 = ScalarParameter() {defaultValue=7, parameterName=LightExponent}
  n45 = DynamicParameter() {parameterIndex=1}
  n46 = Add(A=ActorPositionWS(), B=@n45.RGB)
  n47 = Add(A=WorldPosition() {worldPositionShaderOffset=WPT_ExcludeAllShaderOffsets}.XYZ, B=Multiply(A=VertexNormalWS(), B=NamedRerouteUsage()))
  n48 = ScalarParameter() {defaultValue=16000, parameterName=LightRadius}
  n49 = ScalarParameter() {defaultValue=0.05, parameterName=LightHardness}
  n50 = Power(
   Base=Multiply(A=@n45.A, B=SphereMask(A=@n46, B=@n47, Radius=@n48, Hardness=@n49))
   Exp=ScalarParameter() {defaultValue=7, parameterName=LightExponent}
  )
  n51 = Constant()
  NamedRerouteDeclaration(None=Add(A=Power(Base=@n43, Exp=@n44), B=QualitySwitch(Default=@n50, Low=@n51))) {name=Internal Lights, nodeColor=<rgba 0.6172 0 1 1>}
  NamedRerouteDeclaration(None=@n18) {name=Height Mask Normalized, nodeColor=<rgba 1 0.1875 0 1>}
  n52 = VectorParameter() {parameterName=Heart Glow Color, defaultValue=<rgba 0.0517 2 1.8279 1>}
  n53 = WorldPosition() {worldPositionShaderOffset=WPT_ExcludeAllShaderOffsets}
  n54 = Multiply(A=VertexNormalWS(), B=NamedRerouteUsage())
  n55 = LinearInterpolate(Alpha=@n52.A) {constA=1, constB=2}
  n56 = ScalarParameter() {defaultValue=6000, parameterName=EyeLightRadius}
  n57 = Multiply(
   A=@n52.A
   B=SphereMask(
    A=VectorParameter() {parameterName=Heart Position, defaultValue=<rgba 8000 -8000 -5000 1.25>}.RGB
    B=Add(A=@n53.XYZ, B=@n54)
    Radius=Multiply(A=@n55, B=@n56)
    Hardness=ScalarParameter() {parameterName=EyeLightHardness}
   )
  )
  n58 = ScalarParameter() {defaultValue=8, parameterName=EyeLightExponent}
  NamedRerouteDeclaration(None=Multiply(A=@n52.RGB, B=Power(Base=@n57, Exp=@n58))) {name=Eye Glow, nodeColor=<rgba 0 1 0.0859 1>}
  n59 = Multiply(A=NamedRerouteUsage()) {constB=1.25}
  n60 = ScalarParameter() {defaultValue=2.5, parameterName=RimFresnelExp}
  n61 = TextureCoordinate()
  n62 = Normalize(VectorInput=AppendVector(A=ConstantBiasScale(None=@n61) {bias=-0.5, scale=2}, B=NamedRerouteUsage()))
  n63 = VectorParameter() {defaultValue=<rgba -7000 8000 -10000 1>}
  n64 = VectorParameter() {parameterName=Param_1, defaultValue=<rgba 0 0 -5000 1>}
  n65 = NamedRerouteUsage()
  n66 = Subtract(A=ParticlePositionWS(), B=Add(A=ActorPositionWS(), B=LinearInterpolate(A=@n63.RGB, B=@n64.RGB, Alpha=@n65)))
  n67 = Normalize(VectorInput=Add(A=Transform(None=@n62), B=Normalize(VectorInput=@n66)))
  n68 = Power(
   Base=Multiply(A=Saturate(None=@n59), B=Fresnel(ExponentIn=@n60, Normal=@n67) {baseReflectFraction=0, exponent=4})
   Exp=ScalarParameter() {defaultValue=2, parameterName=RimLightPower}
  )
  n69 = LinearInterpolate(
   A=VectorParameter() {parameterName=RimLightColorMoon, defaultValue=<rgba 1.5 0.1782 0.3796 1>}.RGB
   B=VectorParameter() {parameterName=RimLightColorPortal, defaultValue=<rgba 0.5 0 0.4309 1>}.RGB
   Alpha=NamedRerouteUsage()
  )
  n70 = CameraPositionWS()
  n71 = ActorPositionWS()
  n72 = DotProduct(
   A=Normalize(VectorInput=Subtract(A=@n70, B=@n71))
   B=Transform(None=Constant3Vector() {constant=<rgba 1 0 0 0>}) {transformSourceType=TRANSFORMSOURCE_Local}
  )
  NamedRerouteDeclaration(None=Multiply(A=Multiply(A=@n68, B=@n69), B=Saturate(None=@n72))) {name=RimLight, nodeColor=<rgba 1 0.0362 0.527 1>}
  NamedRerouteDeclaration(None=@n11.OpacityMult) {name=OpacityMult, nodeColor=<rgba 1 0.1172 0 1>}
 ]
}
