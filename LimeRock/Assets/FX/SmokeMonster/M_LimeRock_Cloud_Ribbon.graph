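# Unlit translucent Niagara ribbon and sprite material for LimeRock smoke monster clouds, tinted dark purple.
# Two counter-panning cloud noise samples erode sphere-mask edge and end fades, thresholded by particle alpha, with dynamic parameters controlling panning and end-fade shift.
# That shape drives base shading and opacity, softened by a depth fade offset by particle radius.
# A portal mask function affects color and opacity, and a constructed normal gives sun/moon-directed rim light.
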Material /LimeRock/Assets/FX/SmokeMonster/M_LimeRock_Cloud_Ribbon
graph {
 properties {
  bAutomaticallySetUsageInEditor=false
  bUsedWithNiagaraRibbons=true
  bUsedWithNiagaraSprites=true
  blendMode=BLEND_Translucent
  shadingModel=MSM_Unlit
  translucencyLightingMode=TLM_VolumetricPerVertexDirectional
  usedShadingModels=Unlit
 }
 nodes [
  n1 = TextureCoordinate()
  AppendVector(A=ComponentMask(None=@n1) {g=true}, B=Constant() {r=0.5})
  n2 = Multiply(A=TextureCoordinate() {coordinateIndex=1}, B=ScalarParameter() {defaultValue=0.5, parameterName=CloudTile}) {constB=0.25}
  n3 = DynamicParameter() {paramNames [PanSpeed, PanAdd, OpacityMult, EndFadePowerShift]}
  n4 = Add(A=Time(), B=@n3.PanAdd)
  n5 = Constant2Vector() {g=-0.75, r=0.6}
  n6 = Multiply(
   A=AppendVector(A=ScalarParameter() {defaultValue=0.2, parameterName=Pan X}, B=ScalarParameter() {defaultValue=-0.06, parameterName=Pan Y})
   B=@n3.PanSpeed
  )
  n7 = TextureSample(UVs=Panner(Coordinate=Multiply(A=@n2) {constB=0.831}, Time=@n4, Speed=Multiply(A=@n5, B=@n6) {constA=-0.98}) {speedY=0.1}) {samplerType=SAMPLERTYPE_LinearGrayscale, texture=/Game/Environments/Helios/TechArt/LowAltitudeClouds/T_LowAltitudeClouds_A}
  n8 = TextureSample(UVs=Panner(Coordinate=@n2, Time=@n4, Speed=@n6) {speedY=0.1}) {samplerType=SAMPLERTYPE_LinearGrayscale, texture=&T_LowAltitudeClouds_A}
  LinearInterpolate(A=@n7.G, B=@n8.G)
  n9 = ParticleColor()
  n10 = Constant() {r=0.5}
  n11 = SphereMask(
   A=ComponentMask(None=TextureCoordinate()) {g=true}
   B=@n10
   Radius=@n10
   Hardness=ScalarParameter() {defaultValue=0.25, parameterName=TopBottomFadeHardness}
  ) {hardnessPercent=0}
  n12 = TextureCoordinate()
  n13 = Constant() {r=0.5}
  n14 = SphereMask(
   A=Power(Base=ComponentMask(None=@n12) {r=true}, Exp=@n3.EndFadePowerShift) {constExponent=1}
   B=@n13
   Radius=@n13
   Hardness=ScalarParameter() {defaultValue=0.5, parameterName=EndFadeHardness}
  ) {hardnessPercent=0}
  n15 = OneMinus(None=Max(A=@n7.G, B=@n8.G))
  n16 = ScalarParameter() {defaultValue=0.4, parameterName=NoiseSub}
  n17 = LinearInterpolate(Alpha=@n9.A) {constA=1, constB=0}
  n18 = Subtract(A=Subtract(A=Multiply(A=@n11, B=@n14), B=Multiply(A=@n15, B=@n16)), B=@n17)
  n19 = OneMinus(None=@n17)
  n20 = Power(Base=Saturate(None=Divide(A=@n18, B=@n19)))
  n21 = Multiply(A=@n9.A, B=@n20)
  n22 = NamedRerouteDeclaration(None=@n21) {name=Shape Mask, nodeColor=<rgba 0.6406 0 1 1>}
  n23 = ScalarParameter() {defaultValue=0.7, parameterName=Base Exp}
  n24 = Multiply(
   A=VectorParameter() {parameterName=Base Color, defaultValue=<rgba 0.0384 0.032 0.1 1>}.RGB
   B=LinearInterpolate(
    A=ScalarParameter() {defaultValue=0.3, parameterName=Base 02}
    B=ScalarParameter() {defaultValue=1, parameterName=Base 01}
    Alpha=Power(Base=@n22, Exp=@n23) {constExponent=1.5}
   )
  )
  MaterialFunctionCall(Emissive=Add(A=@n24, B=NamedRerouteUsage())) {materialFunction=&MF_FX_EmissiveAdjustments}
  n25 = MaterialFunctionCall!(MaskEnabled) {materialFunction=/LimeRock/Assets/Materials/MF_LImeRock_PortalOpacityMask}
  n26 = OneMinus(None=NamedRerouteUsage())
  n27 = Multiply(A=ParticleRadius()) {constB=0.5}
  n28 = Subtract(
   A=Subtract(
    A=MaterialFunctionCall!(Fraction Offset, Use Custom Depth, Pixel Offset) {materialFunction=/Engine/Functions/Engine_MaterialFunctions02/SampleSceneDepth}.Depth Using Pixel Offset
    B=Multiply(A=@n26, B=@n27)
   )
   B=PixelDepth().R
  )
  n29 = ScalarParameter() {defaultValue=2000, parameterName=Depth}
  n30 = Multiply(A=@n21, B=Saturate(None=Divide(A=@n28, B=@n29) {constB=500}))
  n31 = ScalarParameter() {defaultValue=0.6, parameterName=Softness}
  n32 = Power(Base=Saturate(None=Divide(A=@n30, B=@n31))) {constExponent=1}
  Multiply(A=@n25.OpacityMask, B=@n32)
  Multiply(A=@n32, B=NamedRerouteUsage())
  Multiply(
   A=@n24
   B=LinearInterpolate(
    B=Constant3Vector() {constant=<rgba 0.923 0.1765 1 1>}
    Alpha=NamedRerouteDeclaration(None=@n25.InteriorMask) {name=InsidePortalMask, nodeColor=<rgba 1 0.75 0 1>}
   ) {constA=1, constB=5}
  )
  NamedRerouteDeclaration(None=@n20) {name=Height Mask Normalized, nodeColor=<rgba 1 0.1875 0 1>}
  n33 = Saturate(None=Multiply(A=NamedRerouteUsage()) {constB=1.25})
  n34 = ConstantBiasScale(None=@n1) {bias=-0.5, scale=2}
  n35 = NamedRerouteUsage()
  n36 = Transform(None=Normalize(VectorInput=AppendVector(A=@n34, B=@n35)))
  n37 = CollectionParameter() {
   collection=&FortniteMaterialParameters
   parameterName=SunAndMoonModelDirectionalVector
  }
  n38 = ComponentMask(None=@n37) {b=true, g=true, r=true}
  n39 = Power(Base=Saturate(None=DotProduct(A=@n36, B=@n38)), Exp=ScalarParameter() {defaultValue=3, parameterName=RimLightPower})
  n40 = LinearInterpolate(
   A=VectorParameter() {parameterName=RimLightColorMoon, defaultValue=<rgba 1.5 0.1782 0.3796 1>}.RGB
   B=VectorParameter() {parameterName=RimLightColorPortal, defaultValue=<rgba 0.5 0 0.4309 1>}.RGB
   Alpha=NamedRerouteUsage()
  )
  NamedRerouteDeclaration(None=Multiply(A=Multiply(A=@n33, B=@n39), B=Multiply(A=@n40) {constB=0.15})) {name=RimLight, nodeColor=<rgba 1 0.0362 0.527 1>}
  NamedRerouteDeclaration(None=@n3.OpacityMult) {name=OpacityMult, nodeColor=<rgba 1 0.1172 0 1>}
 ]
}
