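# Additive unlit Niagara material for the glowing eye of the LimeRock smoke monster.
# A radial glow plus soft flare is broken up by world-space panning Voronoi volume fog, then drives an outer-to-inner color blend and HDR emissive, with a simpler mobile path.
# Opacity combines particle alpha, depth fade and the LimeRock portal mask.
# A switchable world position offset forms a sprite and offsets it relative to the camera.
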Material /LimeRock/Assets/FX/SmokeMonster/M_LimeRock_Monster_Eye
graph {
 properties {
  bAutomaticallySetUsageInEditor=false
  bUsedWithNiagaraMeshParticles=true
  bUsedWithNiagaraSprites=true
  blendMode=BLEND_Additive
  shadingModel=MSM_Unlit
  twoSided=true
  usedShadingModels=Unlit
 }
 nodes [
  n1 = Constant() {r=0.5}
  n2 = SphereMask(A=TextureCoordinate(), B=@n1, Radius=@n1) {hardnessPercent=0}
  n3 = ScalarParameter() {defaultValue=0.5, parameterName=FlareTextureSize}
  n4 = TextureSample(UVs=Saturate(None=MaterialFunctionCall(Texture Scale=@n3) {materialFunction=&ScaleUVsByCenter}.UVs)) {texture=/Game/Effects/Fort_Effects/Textures/Flares/T_Flare_Soft}
  n5 = Add(A=Multiply(A=Power(Base=@n2) {constExponent=3}, B=ScalarParameter() {defaultValue=0.4, parameterName=OuterGlowMult}) {constB=0.5}, B=@n4.RGB)
  n6 = Add(A=WorldPosition().XYZ, B=Multiply(A=Time(), B=VectorParameter() {parameterName=VolumeFogPanSpeed, defaultValue=<rgba 50 -500 1000 0>}.RGB))
  n7 = ScalarParameter() {defaultValue=5000, parameterName=VolumeFogScale}
  n8 = Power(Base=TextureSampleParameterVolume!(UVs=Divide(A=@n6, B=@n7) {constB=500}, Apply View MipBias) {parameterName=Param, samplerType=SAMPLERTYPE_LinearColor, texture=/Game/Effects/FogVolume/VT_Voronoi_LOW}.G)
  n9 = ScalarParameter() {defaultValue=0.07, parameterName=FlareThroughFogAdd}
  n10 = Saturate(None=Add(A=Multiply(A=@n5, B=@n8), B=Multiply(A=@n4.RGB, B=@n9) {constB=0.07}))
  n11 = LinearInterpolate(
   A=VectorParameter() {parameterName=OuterEyeColor, defaultValue=<rgba 0.2 0.0103 0.1445 1>}.RGB
   B=VectorParameter() {parameterName=InnerEyeColor, defaultValue=<rgba 1 0.0052 0.0151 1>}.RGB
   Alpha=@n10
  )
  n12 = Multiply(A=@n11, B=@n10)
  n13 = Multiply(A=ScalarParameter() {defaultValue=4, parameterName=HDRMult}, B=DynamicParameter() {paramNames [HDRMult,Param2,Param3,Param4]}.HDRMult)
  n14 = Constant() {r=0.5}
  MaterialFunctionCall(Emissive=ShadingPathSwitch(
    Default=MaterialFunctionCall(LDR Input=@n12, HDR Tint=@n11, HDR Intensity Multiplier=@n13, Desaturate Luminance=@n14) {materialFunction=&DeriveHDRfromLDR}.RemappedRange
    Mobile=Multiply(A=@n12, B=@n13)
   )) {materialFunction=&MF_FX_EmissiveAdjustments}
  n15 = ScalarParameter() {defaultValue=0.25, parameterName=DepthFadeMult}
  n16 = ScalarParameter() {defaultValue=3000, parameterName=BaseSpriteSize}
  Multiply(
   A=ParticleColor().A
   B=DepthFade(
    Opacity=MaterialFunctionCall!(MaskEnabled) {materialFunction=/LimeRock/Assets/Materials/MF_LImeRock_PortalOpacityMask}.OpacityMask
    FadeDistance=Multiply(A=@n15, B=@n16) {constB=0.5}
   )
  )
  n17 = WorldPosition()
  n18 = Multiply(A=@n16, B=ScalarParameter() {defaultValue=0.125, parameterName=CamOffsetMult})
  StaticSwitchParameter(
   True=Add(
    A=MaterialFunctionCall!(Center Location=@n17.XYZ, 0-1 UVs, XY scale=@n16, OverridePivot, "Normals (Optional)") {materialFunction=&Sprite}.World position offset
    B=MaterialFunctionCall(Offset Amount=@n18) {materialFunction=/Engine/Functions/Engine_MaterialFunctions02/WorldPositionOffset/CameraOffset}.Camera Offset
   )
   False=Constant()
  ) {defaultValue=true, parameterName=UseWPOSprite}
 ]
}
