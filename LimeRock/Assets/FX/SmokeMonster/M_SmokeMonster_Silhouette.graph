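# Unlit masked material for the smoke monster silhouette, built for Niagara mesh particles.
# Emissive multiplies a dark base color by a Fresnel blend of two brightness levels and a portal interior-mask dimming factor, then applies shared FX emissive adjustments.
# Opacity combines the portal mask, a UV sphere mask scaled by vertex red, and a Fresnel edge fade weighted by vertex green.
# DitherTemporalAA turns that opacity into a soft-looking dithered cutout.
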Material /LimeRock/Assets/FX/SmokeMonster/M_SmokeMonster_Silhouette
graph {
 properties {
  bAutomaticallySetUsageInEditor=false
  bUsedWithNiagaraMeshParticles=true
  blendMode=BLEND_Masked
  previewMesh=&/SM_LimeRock_Monster_Head_Silhouette
  shadingModel=MSM_Unlit
  usedShadingModels=Unlit
 }
 nodes [
  n1 = LinearInterpolate(
   A=ScalarParameter() {defaultValue=0.3, parameterName=Base 02}
   B=ScalarParameter() {defaultValue=1, parameterName=Base 01}
   Alpha=Fresnel() {baseReflectFraction=0, exponent=2}
  )
  n2 = MaterialFunctionCall!(MaskEnabled) {materialFunction=/LimeRock/Assets/Materials/MF_LImeRock_PortalOpacityMask}
  n3 = LinearInterpolate(Alpha=@n2.InteriorMask) {constA=1, constB=0.3}
  MaterialFunctionCall(Emissive=Multiply(A=Multiply(A=@n1, B=@n3), B=VectorParameter() {parameterName=Base Color, defaultValue=<rgba 0.0192 0.016 0.05 1>}.RGB) {constA=0.3}) {materialFunction=&MF_FX_EmissiveAdjustments}
  n4 = ScalarParameter() {defaultValue=2, parameterName=EdgeFadExp}
  n5 = MaterialFunctionCall(
   In=OneMinus(None=Fresnel(ExponentIn=@n4) {baseReflectFraction=0, exponent=1.5})
   Contrast=ScalarParameter() {defaultValue=0.5, parameterName=EdgeFadeContrast}
  ) {materialFunction=&CheapContrast}
  n6 = VertexColor()
  n7 = TextureCoordinate()
  n8 = Constant2Vector() {g=0.6, r=0.5}
  n9 = Multiply(A=@n6.R, B=LinearInterpolate(Alpha=SphereMask(A=@n7, B=@n8) {attenuationRadius=0.4, hardnessPercent=0}) {constA=0.25})
  MaterialFunctionCall(Alpha Threshold=Multiply(A=LinearInterpolate(B=@n5.Result, Alpha=@n6.G) {constA=1}, B=Multiply(A=@n2.OpacityMask, B=@n9))) {materialFunction=&DitherTemporalAA}
 ]
}
