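# Additive, unlit, two-sided material for foggy light beams on mesh particles in the SmokeMonster effects.
# A world-space panning Voronoi volume texture, combined with UV end fades and tiled sphere masks, blends emissive from darkened to full particle color and shapes opacity.
# SplineThicken takes base and tip widths scaled by dynamic parameters, supplying normals for a switchable view-angle fade.
# That fade and a length-varying depth fade are skipped on ES2/ES3.1.
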Material /LimeRock/Assets/FX/SmokeMonster/M_Spline_LightBeam_Foggy
graph {
 properties {
  bUsedWithMeshParticles=true
  bUsedWithNiagaraMeshParticles=true
  blendMode=BLEND_Additive
  pixelDepthOffsetMode=PDOM_Legacy
  shadingModel=MSM_Unlit
  twoSided=true
  usedShadingModels=Unlit
 }
 nodes [
  n1 = ParticleColor()
  n2 = VectorParameter() {parameterName=DarkColor, defaultValue=<rgba 0.1935 0.1649 0.4167 0>}
  n3 = Add(A=WorldPosition().XYZ, B=Multiply(A=Time(), B=VectorParameter() {parameterName=VolumeFogPanSpeed, defaultValue=<rgba 50 -500 1000 0>}.RGB))
  n4 = ScalarParameter() {defaultValue=5000, parameterName=VolumeFogScale}
  n5 = TextureCoordinate()
  n6 = LinearInterpolate(
   B=TextureSampleParameterVolume!(UVs=Divide(A=@n3, B=@n4) {constB=500}, Apply View MipBias) {parameterName=Param, samplerType=SAMPLERTYPE_LinearColor, texture=/Game/Effects/FogVolume/VT_Voronoi_LOW}.G
   Alpha=MaterialFunctionCall(In=ComponentMask(None=@n5) {g=true}, Contrast=Constant() {r=0.25}) {materialFunction=&CheapContrast}.Result
  ) {constA=1}
  n7 = TextureCoordinate()
  n8 = ComponentMask(None=@n7) {g=true}
  n9 = LinearInterpolate(A=ScalarParameter() {defaultValue=4, parameterName=BottomFadeSize}, Alpha=@n8) {constA=1, constB=0}
  n10 = ComponentMask(None=@n7) {r=true}
  n11 = ScalarParameter() {defaultValue=1, parameterName=BeamTile}
  n12 = Constant() {r=0.5}
  n13 = SphereMask(A=Fmod(A=Multiply(A=@n10, B=@n11), B=Constant() {r=1}), B=@n12, Radius=@n12) {hardnessPercent=0}
  n14 = Multiply(A=Multiply(A=@n6, B=@n6), B=Multiply(A=Saturate(None=@n9), B=Multiply(A=@n8, B=@n13)))
  MaterialFunctionCall(Emissive=LinearInterpolate(A=Multiply(A=@n1.RGB, B=@n2.RGB), B=@n1.RGB, Alpha=@n14)) {materialFunction=&MF_FX_EmissiveAdjustments}
  n15 = ScalarParameter() {defaultValue=1, parameterName=OpacityPower}
  n16 = DynamicParameter() {paramNames [WidthBaseMult, WidthTipMult, FlickerFade, FlickerSpeed]}
  n17 = Multiply(A=@n16.WidthBaseMult, B=ScalarParameter() {defaultValue=5, parameterName=WidthBase})
  n18 = Multiply(A=@n16.WidthTipMult, B=ScalarParameter() {defaultValue=100, parameterName=WidthTip})
  n19 = StaticSwitchParameter(True=Multiply(A=VertexColor().R, B=@n18), False=@n18) {defaultValue=true, parameterName=VertexColorTipWidthOffset}
  n20 = ScalarParameter() {defaultValue=0.5, parameterName=AngleFade}
  n21 = Power(
   Base=DotProduct(
    A=CameraVectorWS()
    B=MaterialFunctionCall(WidthBase=@n17, WidthTip=@n19) {materialFunction=&SplineThicken}.Normal
   )
   Exp=@n20
  )
  n22 = ScalarParameter() {defaultValue=500, parameterName=DepthTipDist}
  n23 = ScalarParameter() {defaultValue=200, parameterName=DepthBaseDist}
  n24 = ComponentMask(None=TextureCoordinate()) {g=true}
  n25 = Multiply(
   A=StaticSwitchParameter(
    True=Clamp(None=@n21)
    False=MaterialFunctionCall("FadeSensitivity (optional)"=@n20) {materialFunction=&FadeBasedOnViewAngle}.Result
   ) {parameterName=Use WPO Adjusted Normals for Angle Fade}
   B=DepthFade(FadeDistance=LinearInterpolate(A=@n22, B=@n23, Alpha=@n24))
  )
  n26 = Constant() {r=1}
  Multiply(A=@n1.A, B=Multiply(A=Power(Base=@n14, Exp=@n15), B=FeatureLevelSwitch(Default=@n25, ES2=@n26, ES3_1=@n26, SM4_REMOVED=@n25)))
 ]
}
