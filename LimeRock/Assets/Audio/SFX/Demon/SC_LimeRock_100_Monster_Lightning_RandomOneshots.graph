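# Sound cue for one-shot demon monster lightning sound effects in LimeRock.
# A Random node picks one of four equally weighted lightning waves, then a Modulator varies its pitch slightly.
# It plays at reduced volume through a lightning attenuation asset and the Fort_Events sound class.
# An overridden concurrency limits simultaneous instances per owner, with a retrigger time and lowest-priority voices stopped first.
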SoundCue /LimeRock/Assets/Audio/SFX/Demon/SC_LimeRock_100_Monster_Lightning_RandomOneshots
root=Modulator_0
nodes [
 {id=Modulator_0, children [Random_1], properties {pitchMax=1.1, pitchMin=0.9}}
 {id=Random_1, children [WavePlayer_5,WavePlayer_6,WavePlayer_7,WavePlayer_8], properties {weights [1,1,1,1]}}
 {id=WavePlayer_5, properties {soundWaveAssetPtr=/LimeRock/Assets/Audio/SFX/Demon/Wav/LimeRock_100_Monster_Lightning_01}}
 {id=WavePlayer_6, properties {soundWaveAssetPtr=/LimeRock/Assets/Audio/SFX/Demon/Wav/LimeRock_100_Monster_Lightning_02}}
 {id=WavePlayer_7, properties {soundWaveAssetPtr=/LimeRock/Assets/Audio/SFX/Demon/Wav/LimeRock_100_Monster_Lightning_03}}
 {id=WavePlayer_8, properties {soundWaveAssetPtr=/LimeRock/Assets/Audio/SFX/Demon/Wav/LimeRock_100_Monster_Lightning_04}}
]
properties {
 attenuationSettings=/LimeRock/Assets/Audio/Settings/ATT_LimeRock_Lightning
 bOverrideConcurrency=true
 maxDistance=140000
 soundClassObject=&Fort_Events
 volumeMultiplier=0.4
 concurrencyOverrides {bLimitToOwner=true, maxCount=3, resolutionRule=StopLowestPriority, retriggerTime=0.5}
 concurrencySet [null]
}
