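# Niagara effect for distant pink-magenta lightning on the horizon, part of the SmokeMonster effects, with randomized monster lightning sounds.
# Timed bursts plus probabilistic spawning driven by User.SpawnRate place short-lived strikes far from the origin, with angles from a partition array or random curve, selected by spawn group.
# Each strike drives mesh, sprite and light renderers.
# Curves animate two-phase flashes, shifting the sprite along the bolt at mid-life, while waveforms flicker light and sprite size.
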NiagaraSystem /LimeRock/Assets/FX/SmokeMonster/NS_LimeRock_HorizonLightning
dependencies {
 usedDataInterfaces [&NiagaraDataInterfaceAudioPlayer, &NiagaraDataInterfaceArrayFloat, &NiagaraDataInterfaceCurve]
 usedDynamicInputs []
 usedModules [&SystemState]
}
emitters {HorizonLightning {
  emitterSpawnScript [&InitializeAudioPlayer(Audio Player=DataInterface() {propertyValues {
      Attenuation=null
      Concurrency=null
      MaxPlaysPerTick=10
      SoundToPlay=/LimeRock/Assets/Audio/SFX/Demon/SC_LimeRock_100_Monster_Lightning_RandomOneshots
      bAllowLoopingOneShotSounds=false
      bLimitPlaysPerTick=true
      bOnlyActiveDuringGameplay=false
      bStopWhenComponentIsDestroyed=true
      ConfigurationUserParameter {parameter {name=None, typeDefHandle {registeredTypeIndex=1}, varData []}}
      ParameterNames []
     }})]
  emitterUpdateScript [
   &EmitterState(Life Cycle Mode=::NewEnumerator1, Loop Behavior=::NewEnumerator1, Loop Duration Mode=::NewEnumerator1, UseLoopDelay=false)
   &SpawnBurst_Instantaneous(Use Spawn Probability=false, UseLoopCountLimit=false)
   &SpawnBurst_Instantaneous(Spawn Time=0.77, Use Spawn Probability=false, UseLoopCountLimit=false) {module=SpawnBurst_Instantaneous001}
   &SpawnBurst_Instantaneous(Spawn Time=1.27, Use Spawn Probability=false, UseLoopCountLimit=false) {module=SpawnBurst_Instantaneous002}
   &SpawnBurst_Instantaneous(Spawn Time=1.5, Use Spawn Probability=false, UseLoopCountLimit=false) {module=SpawnBurst_Instantaneous003}
   &SpawnBurst_Instantaneous(Spawn Time=2, Use Spawn Probability=false, UseLoopCountLimit=false) {module=SpawnBurst_Instantaneous004}
   &SpawnRate(Spawn Probability=0.3, SpawnGroup=1, SpawnRate=$User.SpawnRate, Use Spawn Probability=true)
  ]
  particleSpawnScript [
   &PartitionParticles(Number Of Partitions=4)
   &InitializeParticle(
    Color=&RandomRangeLinearColor(Minimum=<rgba 1.361 0.0104 2 1>, Maximum=<rgba 3 0.0469 0.7591 1>)
    Lifetime=&RandomRangeFloat(Minimum=0.2, Maximum=0.6)
    Mesh Scale Max=<100 120 120>
    Mesh Scale Min=<100 50 50>
    Mesh Scale Mode=::NewEnumerator2
    Position Offset=&RotateVectorByQuaternion(
     Vector To Rotate=<-70000 0 0>
     Quaternion=&MakeQuaternion(
      Quaternion From=0
      Axis=<0 0 1>
      Angle=&MaskFloatBySpawnGroup(
       Masked Float=&SelectFloatFromArray(
        Direct Array Index=$Particles.PartitionParticles.Partition
        Direct Array Mode=::NewEnumerator1
        Float Selection Array=DataInterface() {propertyValues {FloatData [-0.07,0.2,-0.14,0.22]}}
       )
       Passthrough Float=&FloatFromCurve(
        FloatCurve=DataInterface() {propertyValues {Curve {
           defaultValue=1
           keys [
            {arriveTangent=0.8275, interpMode=RCIM_Cubic, leaveTangent=0.8275, tangentMode=RCTM_User, time=0, value=-1}
            {arriveTangent=1.9392, tangentMode=RCTM_User, time=0.49, value=-0.3}
            {arriveTangent=2.549, interpMode=RCIM_Cubic, leaveTangent=2.549, time=0.5, value=0.3}
            {arriveTangent=0.6344, interpMode=RCIM_Cubic, leaveTangent=0.6344, tangentMode=RCTM_User, time=1, value=1}
           ]
          }}}
        CurveIndex=&RandomRangeFloat(Minimum=0, Maximum=1)
        Scale Curve=0.25
       )
      )
      From=?
      To=?
      Angle Type=::NewEnumerator1
      Normalize Vectors=false
     )
    )
    Position Offset Coordinate Space=::Local
    Sprite Size Mode=::NewEnumerator1
    Uniform Sprite Size Max=80000
    Uniform Sprite Size Min=60000
    UsePositionOffset=true
    Write Material Random=true
   ) {module=InitializeParticle001}
   &ShapeLocation(
    Cylinder Height=3000
    Cylinder Height Midpoint=0.7
    Cylinder Radius=5000
    MaskSpawnGroup=false
    Override Seed=false
    Shape Primitive=::NewEnumerator1
    Surface Only=false
    UseOwnerScale=true
   )
   &InitialMeshOrientation(
    Mesh Orientation Mode=::NewEnumerator2
    Orientation Vector=<0 0 1>
    Rotation=&RandomRangeVector(Minimum=<0 -0.05 -0.05>, Maximum=<1 0.05 0.05>)
    Use Rotation Vector=true
   )
   /SetVariables(Particles.MeshIndex=&RandomRangeInt(Minimum=0, Maximum=1)) {isSetParameters=true}
   /Niagara/Modules/Audio/PlayAudio()
  ]
  particleUpdateScript [
   &ParticleState()
   &ScaleColor(
    Scale Alpha=&FloatFromCurve(FloatCurve=DataInterface() {propertyValues {Curve {keys [
         <at 0 0>
         {interpMode=RCIM_Cubic, tangentMode=RCTM_User, time=0.1258, value=0.1952}
         {arriveTangent=0.1568, interpMode=RCIM_Cubic, leaveTangent=0.1568, time=0.341, value=0.0902}
         {interpMode=RCIM_Cubic, tangentMode=RCTM_User, time=0.496, value=0.2532}
         <at 1 0>
        ]}}})
    Scale RGB=&VectorFromFloat(Value=&FloatFromCurve(FloatCurve=DataInterface() {propertyValues {Curve {keys [
          <at 0 0.5>
          {arriveTangent=7, interpMode=RCIM_Cubic, leaveTangent=7, time=0.45, value=1}
          {arriveTangent=-4.524, interpMode=RCIM_Cubic, leaveTangent=-4.524, tangentMode=RCTM_User, time=0.5, value=4}
          <at 1 1>
         ]}}}))
    ScaleA=true
    ScaleRGB=true
   )
   &DynamicMaterialParameters(
    Index 0 Param 1=&FloatFromCurve(FloatCurve=DataInterface() {propertyValues {Curve {keys [
         {arriveTangent=-0.2542, interpMode=RCIM_Cubic, leaveTangent=-0.2542, tangentMode=RCTM_User, time=0, value=1}
         {arriveTangent=-1.6922, interpMode=RCIM_Cubic, leaveTangent=-1.6922, tangentMode=RCTM_User, time=0.4987, value=0.5013}
         {arriveTangent=-0.9805, interpMode=RCIM_Cubic, leaveTangent=-0.9805, tangentMode=RCTM_User, time=1, value=0}
        ]}}})
    Index 0 Param 2=&FloatFromCurve(
     FloatCurve=DataInterface() {propertyValues {Curve {keys [
         {interpMode=RCIM_Cubic, leaveTangent=-1.1405, tangentMode=RCTM_Break, time=0, value=0.5}
         {arriveTangent=4.4936, interpMode=RCIM_Cubic, leaveTangent=-5.3647, tangentMode=RCTM_Break, time=0.4833, value=1}
         <at 1 0>
        ]}}}
     Scale Curve=&RandomRangeFloat(Minimum=1, Maximum=2)
    )
    Index 0 Param 3=&FloatFromCurve(FloatCurve=DataInterface() {propertyValues {Curve {keys [
         <at 0 1 RCIM_Cubic>
         {arriveTangent=-3.8306, interpMode=RCIM_Cubic, tangentMode=RCTM_Break, time=0.49, value=0.1}
         {arriveTangent=-4.2365, interpMode=RCIM_Cubic, leaveTangent=-4.2365, tangentMode=RCTM_User, time=0.5, value=1}
         <at 1 0.2 RCIM_Cubic>
        ]}}})
    Index 0 Param 4=&RandomRangeFloat(Minimum=0, Maximum=1)
    Index 3 Param 1=3
    Index 3 Param 2=1000
    Module.Index0Float0Write=true
    Module.Index0Float1Write=true
    Module.Index0Float2Write=true
    Module.Index0Float3Write=true
    Module.Index3Float0Write=true
    Module.Index3Float1Write=true
    Module.Index3Float2Write=true
    Module.Index3Float3Write=true
    Write Parameter Index 3=::NewEnumerator1
   ) {module=DynamicMaterialParameters001}
   /SetVariables(Particles.SpriteColor=&MakeLinearColorFromVectorAndFloat(
     "Vector (RGB)"=&MakeVectorFromLinearColorRGB(Color=$Particles.Initial.Color)
     "Float (Alpha)"=&FloatFromCurve(
      FloatCurve=DataInterface() {propertyValues {Curve {keys [
          {arriveTangent=16.6867, interpMode=RCIM_Cubic, leaveTangent=16.6867, tangentMode=RCTM_User, time=0, value=0}
          {arriveTangent=-4.0106, interpMode=RCIM_Cubic, leaveTangent=-4.0106, tangentMode=RCTM_User, time=0.125, value=1}
          {arriveTangentWeight=4.8377, interpMode=RCIM_Cubic, leaveTangent=32.5316, leaveTangentWeight=0.607, tangentMode=RCTM_Break, time=0.5, value=0}
          {arriveTangent=-1.5315, interpMode=RCIM_Cubic, leaveTangent=-1.5315, tangentMode=RCTM_User, time=0.56, value=0.7}
          <at 1 0>
         ]}}}
      Scale Curve=0.25
     )
    )) {isSetParameters=true}
   /SetVariables(Particles.LightColor=&MultiplyLinearColor(
     Linear Color=$Particles.Color
     Scale Factor=&MakeVector4_FromFloat(Float=&Waveform(Global Amplitude Scale=2, "Waveform [1]"=::NewEnumerator12, "[1] Frequency"=30, "[1] Amplitude Min/Max"=<0.1 1>))
    )) {isSetParameters=true}
   /SetVariables(Particles.SpritePosition:NiagaraPosition=&AddVectorToPosition(
     Position=$Particles.Initial.Position
     Vector=&MakeCustomVectorFromBool(
      Bool=&SetBoolByFloatComparison(A=$Particles.NormalizedAge, B=0.5, Comparison Type=::NewEnumerator2)
      True Vector=<0 0 0>
      False Vector=&Multiply_VectorByFloat(
       Vector=&ReturnMeshOrientationAxis(Mesh Orientation Axis=::XAxis, Mesh Orientation Quat=$Particles.Initial.MeshOrientation)
       Float=&Multiply_Float(A=&MakeFloatFromVector(Vector=$Particles.Scale, Channel=::NewEnumerator0), B=1000)
      )
     )
    )) {isSetParameters=true}
   /SetVariables(Particles.LightPosition:NiagaraPosition=&AddVectorToPosition(
     Position=&LerpPosition(A=$Particles.SpritePosition, B=$Engine.Owner.Position, Alpha=0.3)
     Vector=<4000 0 -4000>
    )) {isSetParameters=true}
   &ScaleSpriteSize(
    Scale Sprite Size Mode=::NewEnumerator0
    Uniform Scale Factor=&FloatFromCurve(
     FloatCurve=DataInterface() {propertyValues {Curve {keys [<at 0.4 1 RCIM_Cubic>,<at 0.7 0.2>]}}}
     Scale Curve=&Waveform("Waveform [1]"=::NewEnumerator12, "[1] Frequency"=30, "[1] Amplitude Min/Max"=<0.5 1.1>)
    )
   )
   &SpriteFacingAndAlignment(
    Facing Coordinate Space=::World
    Sprite Alignment=<0 0 -1>
    Sprite Facing=&Lerp_Vector(
     A=<0 0 -1>
     B=$System.PositionInFrontOfCam
     Alpha=&MakeCustomFloatFromBool(
      Bool=&SetBoolByFloatComparison(A=$Particles.NormalizedAge, B=0.5, Comparison Type=::NewEnumerator2)
      True Float=1
      False Float=0.1
     )
    )
    WriteAlignment=false
    WriteFacing=true
   )
  ]
  rendererClasses [
   &NiagaraMeshRendererProperties
   &NiagaraSpriteRendererProperties
   &NiagaraLightRendererProperties
  ]
  renderers [
   {rendererClass=&NiagaraMeshRendererProperties, rendererIndex=0}
   {rendererClass=&NiagaraSpriteRendererProperties, rendererIndex=1}
   {rendererClass=&NiagaraLightRendererProperties, rendererIndex=2}
  ]
 }}
summary {
 systemName=NS_LimeRock_HorizonLightning
 emitters [{
   emitterName=HorizonLightning
   rendererClasses [
    &NiagaraMeshRendererProperties
    &NiagaraSpriteRendererProperties
    &NiagaraLightRendererProperties
   ]
  }]
 userVariables [User.SpawnRate {type=NiagaraFloat, defaultValue=0, description=""}]
}
