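# Niagara system for the LimeRock smoke monster: a towering cloud body with a head, heart-glow beams and internal lightning.
# MainSmoke recycles burst sprites by wrapped age along tilting torus rings whose radius follows User.Spread, fading and shrinking them by camera facing.
# User.HeartGlow drives head glow and beam brightness; beams radiate from System.HeartPosition.
# Lightning spawns at two System light positions also fed to smoke materials; big bolts depend on Spread.
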NiagaraSystem /LimeRock/Assets/FX/SmokeMonster/NS_LimeRock_SmokeMonsterMain
dependencies {
 usedDataInterfaces [
  &NiagaraDataInterfaceCurve
  &NiagaraDataInterfaceArrayFloat
  &NiagaraDataInterfaceArrayColor
  &NiagaraDataInterfaceMeshRendererInfo
  &NiagaraDataInterfaceVectorCurve
  &NiagaraDataInterfaceCamera
 ]
 usedDynamicInputs [
  &Sine
  &Cosine
  /Niagara/DynamicInputs/Vectors/RotateVector
  &SelectLinearColorFromArray
  &MakeVector4FromVectorAndFloat
  &Abs_Float
  &ReturnCameraVector
  &Add_Vector
  &MakeVectorFromLinearColorRGB
 ]
 usedModules [
  &self:SystemSpawnScript.SystemScriptSource.SystemScriptGraph.NiagaraNodeAssignment_6.SetVariables_#1
  &self:SystemSpawnScript.SystemScriptSource.SystemScriptGraph.NiagaraNodeAssignment_5.SetVariables_#2
  &self:SystemSpawnScript.SystemScriptSource.SystemScriptGraph.NiagaraNodeAssignment_7.SetVariables_#3
  &self:SystemSpawnScript.SystemScriptSource.SystemScriptGraph.NiagaraNodeAssignment_8.SetVariables_#4
  &self:SystemSpawnScript.SystemScriptSource.SystemScriptGraph.NiagaraNodeAssignment_11.SetVariables_#5
  &self:SystemSpawnScript.SystemScriptSource.SystemScriptGraph.NiagaraNodeAssignment_12.SetVariables_#6
  &SystemState
  &Timeline
  &self:SystemSpawnScript.SystemScriptSource.SystemScriptGraph.NiagaraNodeAssignment_13.SetVariables_#7
  &self:SystemSpawnScript.SystemScriptSource.SystemScriptGraph.NiagaraNodeAssignment_19.SetVariables_#8
  &self:SystemSpawnScript.SystemScriptSource.SystemScriptGraph.NiagaraNodeAssignment_20.SetVariables_#9
  &self:SystemSpawnScript.SystemScriptSource.SystemScriptGraph.NiagaraNodeAssignment_21.SetVariables_#10
  &self:SystemSpawnScript.SystemScriptSource.SystemScriptGraph.NiagaraNodeAssignment_4.SetVariables_#11
  &self:SystemSpawnScript.SystemScriptSource.SystemScriptGraph.NiagaraNodeAssignment_14.SetVariables_#12
  &self:SystemSpawnScript.SystemScriptSource.SystemScriptGraph.NiagaraNodeAssignment_17.SetVariables_#13
 ]
}
emitters {
 Beams {
  emitterUpdateScript [
   &EmitterState()
   &SpawnRate(SpawnRate=12, Use Spawn Probability=false)
  ]
  particleSpawnScript [
   &InitializeParticle(
    Color=&RandomRangeLinearColor(Minimum=<rgba 0.1149 0.0376 1.5 0.3>, Maximum=<rgba 0.1301 0.2028 2 0.4>)
    Lifetime=&RandomRangeFloat(Minimum=0.9, Maximum=1.4)
    Mesh Scale Mode=::NewEnumerator3
    Mesh Uniform Scale=&RandomRangeFloat(Minimum=70, Maximum=90)
    UsePositionOffset=false
    Write Material Random=true
   )
   &DynamicMaterialParameters(
    Index 0 Param 1=1
    Index 0 Param 3=0.25
    Index 0 Param 4=1
    Module.Param0WriteEnabled=true
    Module.Param1WriteEnabled=false
    Module.Param2WriteEnabled=true
    Module.Param3WriteEnabled=true
   )
   &CameraOffset(Camera Offset Amount=6)
  ]
  particleUpdateScript [
   &ParticleState()
   &ShapeLocation(
    MaskSpawnGroup=false
    Offset=<0 0 1000>
    Offset Mode=::NewEnumerator0
    Scale Mode:ENiagara_ScaleMode=::NewEnumerator1
    Shape Origin:Vector3f=$System.HeartPosition
    Sphere Distribution=::NewEnumerator1
    Sphere Radius=200
    U Position=&Add_Float(
     A=&RandomRangeFloat(Minimum=-0.25, Maximum=0.25)
     B=&Multiply_Float(A=$Particles.Age, B=0.05)
    )
    V Position=&Add_Float(
     A=&RandomRangeFloat(Minimum=0, Maximum=1)
     B=&Multiply_Float(A=$Particles.Age, B=-0.12)
    )
   )
   &UpdateMeshOrientation(
    Facing Direction=$Particles.ShapeLocation.ShapeNormal
    Facing Interpolation=false
    Orientation Method=::NewEnumerator1
    Override Quat=false
    TransformFacing=false
    TransformUp=false
   )
   /SetVariables(Particles.ForwardFacing:NiagaraFloat=&FloatFromCurve(
     FloatCurve=DataInterface() {propertyValues {Curve {defaultValue=1, keys [{arriveTangent=1.5721, interpMode=RCIM_Cubic, time=0.5, value=0.5}, {tangentMode=RCTM_User, time=1, value=1}]}}}
     CurveIndex=&DotProduct(
      Vector A=$StackContext.ShapeLocation.ShapeNormal
      Vector B=&NormalizeVector(
       Vector To Normalize=&Multiply_Vector(A=$Engine.Owner.SystemXAxis, B=<1 1 0>)
       Safe Normalize=true
       Safe Fallback Threshold=0
       Safe Fallback Vector=<0 0 0.001>
      )
     )
     Scale Curve=$User.HeartGlow
    )) {isSetParameters=true}
   &ScaleMeshSize(Scale Factor=&VectorFromFloat(Value=&FloatFromCurve(
      FloatCurve=DataInterface() {propertyValues {Curve {keys [
          {arriveTangent=1.9706, interpMode=RCIM_Cubic, leaveTangent=1.9706, tangentMode=RCTM_User, time=0, value=0.3}
          {arriveTangent=0.1105, interpMode=RCIM_Cubic, leaveTangent=0.1105, tangentMode=RCTM_User, time=0.5, value=1}
          {arriveTangent=-1.8994, interpMode=RCIM_Cubic, leaveTangent=-1.8994, tangentMode=RCTM_User, time=1, value=0.3}
         ]}}}
      Scale Curve=$Particles.ForwardFacing
     )))
   &ScaleColor(
    Scale Alpha=&FloatFromCurve(
     FloatCurve=DataInterface() {propertyValues {Curve {keys [{interpMode=RCIM_Cubic, tangentMode=RCTM_User, time=0, value=0}, <at 0.3 1 RCIM_Cubic>, {tangentMode=RCTM_User, time=1, value=0}]}}}
     Scale Curve=$Particles.ForwardFacing
    )
    Scale RGB=&VectorFromFloat(Value=&FloatFromCurve(FloatCurve=DataInterface() {propertyValues {Curve {keys [<at 1 1 RCIM_Cubic>,<at 1.25 5>]}}}, CurveIndex=$User.HeartGlow))
    ScaleA=true
    ScaleRGB=true
   )
   &DynamicMaterialParameters(
    Index 0 Param 1=5
    Index 0 Param 2=&FloatFromCurve(
     FloatCurve=DataInterface() {propertyValues {Curve {keys [
         {arriveTangent=3.9743, interpMode=RCIM_Cubic, leaveTangent=3.9743, tangentMode=RCTM_User, time=0, value=0.25}
         {interpMode=RCIM_Cubic, tangentMode=RCTM_User, time=0.4, value=1}
         {arriveTangent=-1.7788, interpMode=RCIM_Cubic, leaveTangent=-1.7788, tangentMode=RCTM_User, time=1, value=0}
        ]}}}
     Scale Curve=20
    )
    Module.Index0Float0Write=true
    Module.Index0Float1Write=true
    Module.Index0Float2Write=false
    Module.Index0Float3Write=false
   ) {module=DynamicMaterialParameters001}
  ]
  rendererClasses [&NiagaraMeshRendererProperties]
  renderers [{rendererClass=&NiagaraMeshRendererProperties, rendererIndex=0}]
 }
 Head {
  emitterUpdateScript [
   &EmitterState(Life Cycle Mode=::NewEnumerator1, Loop Behavior=::NewEnumerator1, Loop Duration Mode=::NewEnumerator1, UseLoopDelay=false)
   &SpawnBurst_Instantaneous(Use Spawn Probability=false, UseLoopCountLimit=false)
  ]
  particleSpawnScript [
   &InitializeParticle(
    Color=<rgba 0.0491 0.0406 0.08 3>
    Lifetime=5
    Mass=5000
    Mass Mode=::NewEnumerator1
    Position Offset=<0 0 500>
    Position Offset Coordinate Space=::Local
    Sprite Rotation Angle=-0.25
    Sprite Rotation Mode=::NewEnumerator1
    Sprite Size Mode=::NewEnumerator3
    Uniform Sprite Size=7000
    UsePositionOffset=true
    Write Material Random=true
   )
   &DynamicMaterialParameters(
    Index 0 Param 1=1
    Index 0 Param 2=5
    Module.Index0Float0Write=true
    Module.Index0Float1Write=true
    Module.Index0Float2Write=false
    Module.Index0Float3Write=false
   )
   &InitialMeshOrientation(Mesh Orientation Mode=::NewEnumerator2, Orientation Vector=<1 0 -0.2>, Use Rotation Vector=false)
   /SetVariables(
    Particles.HeadSmokeColor:LinearColor=<rgba 1 0.0739 0.4583 1>
    Particles.HeadSmokeSize:Vector2f=&Vector2DFromFloat(Value=&Multiply_Float(A=$System.SmokeCloudSize, B=0.9))
   ) {isSetParameters=true}
  ]
  particleUpdateScript [
   &ParticleState(Kill Particles When Lifetime Has Elapsed=false)
   &ScaleSpriteSize(Scale Sprite Size Mode=::NewEnumerator0, Uniform Scale Factor=$User.HeartGlow)
   /SetVariables(Particles.HeartGlowPosition:NiagaraPosition=&AddVectorToPosition(
     Position=&TransformPosition(
      Position=&AddVectorToPosition(Position=Linked() {linkedVariable {name=System.HeartPosition, type=Vector3f}}, Vector=<0 0 1000>)
      Source Space=::World
      Destination Space=::Local
     )
     Vector=<200 0 0>
    )) {isSetParameters=true}
   /SetVariables(Particles.HeartGlowColor:LinearColor=&MakeLinearColorFromVectorAndFloat("Vector (RGB)"=<1 1 1>, "Float (Alpha)"=$User.HeartGlow)) {isSetParameters=true}
   &DynamicMaterialParameters(
    Index 0 Param 1=&Add_Float(
     A=&FloatFromCurve(FloatCurve=DataInterface() {propertyValues {Curve {keys [<at 1 1 RCIM_Cubic>,<at 1.25 5>]}}}, CurveIndex=$User.HeartGlow)
     B=&FloatFromCurve(
      FloatCurve=DataInterface() {propertyValues {Curve {keys [
          {arriveTangent=-0.6714, interpMode=RCIM_Cubic, time=3, value=0}
          {arriveTangent=1.1952, interpMode=RCIM_Cubic, leaveTangent=1.1952, tangentMode=RCTM_User, time=3.5, value=1}
          <at 4.5 0>
         ]}}}
      CurveIndex=$Engine.System.Age
      Scale Curve=4
     )
    )
    Module.Index0Float0Write=true
    Module.Index0Float1Write=false
    Module.Index0Float2Write=false
    Module.Index0Float3Write=true
    Module.Index1Float0Write=true
    Module.Index1Float1Write=true
    Module.Index1Float2Write=true
    Module.Index1Float3Write=true
    Write Parameter Index 1=::NewEnumerator1
   ) {module=DynamicMaterialParameters001}
   &ScaleColor(
    Scale Alpha=&FloatFromCurve(
     FloatCurve=DataInterface() {propertyValues {Curve {keys [<at 3 0 RCIM_Cubic>, {interpMode=RCIM_Cubic, leaveTangent=0.5, time=3.5, value=1}]}}}
     CurveIndex=$Engine.System.Age
    )
    ScaleA=true
    ScaleRGB=false
   )
  ]
  rendererClasses [&NiagaraSpriteRendererProperties, &NiagaraMeshRendererProperties]
  renderers [
   {rendererClass=&NiagaraSpriteRendererProperties}
   {rendererClass=&NiagaraMeshRendererProperties, rendererIndex=1}
   {rendererClass=&NiagaraSpriteRendererProperties, rendererIndex=2}
  ]
 }
 Lightning {
  emitterUpdateScript [
   &EmitterState()
   &SpawnBurst_Instantaneous(
    Spawn Count=&RandomRangeInt(Minimum=0, Maximum=2)
    Spawn Time=0.03
    Use Spawn Probability=false
    UseLoopCountLimit=false
   )
   &SpawnBurst_Instantaneous(
    Spawn Count=&RandomRangeInt(Minimum=0, Maximum=1)
    Spawn Group=1
    Spawn Time=0.33
    Use Spawn Probability=false
    UseLoopCountLimit=false
   ) {module=SpawnBurst_Instantaneous001}
  ]
  particleSpawnScript [
   &InitializeParticle(
    Color=&RandomRangeLinearColor(Minimum=<rgba 1 1 1 0.2>, Maximum=<rgba 1.2 1.2 1.2 0.5>)
    Lifetime=&RandomRangeFloat(Minimum=0.16, Maximum=0.25)
    Mesh Scale Mode=::NewEnumerator3
    Mesh Uniform Scale=&RandomRangeFloat(Minimum=15, Maximum=25)
    Position Offset=&MaskVectorBySpawnGroup(
     Masked Vector=&self:Vector4to3(Vector 4:Vector4f=$System.Light_01)
     Passthrough Vector=&self:Vector4to3(Vector 4:Vector4f=$System.Light_02)
    )
    Sprite Size Mode=::NewEnumerator3
    Uniform Sprite Size=3000
    UsePositionOffset=true
    Write Material Random=true
   )
   &ShapeLocation(MaskSpawnGroup=false, Override Seed=false, Sphere Radius=1500, Sphere Surface Distribution=1, UseOwnerScale=true)
   &AddVelocity(
    Module.Use Falloff Distance=false
    Origin Offset=&RandomRangeVector(Minimum=<-15000 -500 0>, Maximum=<-15000 500 -1000>)
    Origin Offset Coordinate Space=::World
    Override Seed=false
    Velocity Mode=::NewEnumerator1
    Velocity Origin=&SubtractVectorFromPosition(
     Position=$Engine.Owner.Position
     Vector=&MakeVector(X=0, Y=0, Z=&Multiply_Float(A=$System.MaxHeight, B=0.9))
    )
    Velocity Speed=5
   )
   &InitialMeshOrientation(
    Mesh Axis To Orient=<0 0 1>
    Mesh Orientation Mode=::NewEnumerator2
    Rotation=&RandomRangeVector(Minimum=<0.05 0.05 0>, Maximum=<-0.05 -0.05 1>)
    Use Rotation Vector=true
   )
  ]
  particleUpdateScript [
   &ParticleState()
   &ScaleMeshSize(Scale Factor=&MakeVector(
     X=1
     Y=1
     Z=&FloatFromCurve(
      FloatCurve=DataInterface() {propertyValues {Curve {keys [
          {interpMode=RCIM_Cubic, leaveTangent=0.0966, tangentMode=RCTM_Break, time=0, value=1}
          {arriveTangent=0.0951, interpMode=RCIM_Cubic, leaveTangent=-0.1745, tangentMode=RCTM_Break, time=1, value=1.1}
         ]}}}
      Scale Curve=&RandomRangeFloat(Minimum=0.8, Maximum=1.5)
     )
    ))
   &DynamicMaterialParameters(
    Index 0 Param 4=&SelectFloatFromArray(
     Array Sampling Mode=::NewEnumerator0
     Direct Array Index=?
     Float Selection Array=DataInterface() {propertyValues {FloatData [3,0,10,3,1,9,5]}}
     Random Seed=0
    )
    Index 1 Param 1=1
    Index 1 Param 2=2
    Index 1 Param 3=&FloatFromCurve(FloatCurve=DataInterface() {propertyValues {Curve {keys [{interpMode=RCIM_Cubic, tangentMode=RCTM_Break, time=0.5, value=-0.5}, <at 1 0.1>]}}})
    Index 2 Param 1=&Multiply_Float(
     A=&FloatFromCurve(
      FloatCurve=DataInterface() {propertyValues {Curve {keys [{interpMode=RCIM_Cubic, leaveTangent=8.1085, tangentMode=RCTM_Break, time=0.8568, value=1}, {tangentMode=RCTM_User, time=1, value=0}]}}}
      Scale Curve=0.5
     )
     B=&Divide_Float(
      A=&MakeFloatFromVector(Vector=$Particles.Scale, Channel=::NewEnumerator0)
      B=&RandomRangeFloat(Minimum=10, Maximum=15)
     )
    )
    Index 2 Param 2=&FloatFromCurve(
     FloatCurve=DataInterface() {propertyValues {Curve {keys [
         {interpMode=RCIM_Cubic, tangentMode=RCTM_Break, time=0.7, value=0}
         {arriveTangent=-0.0859, interpMode=RCIM_Cubic, tangentMode=RCTM_Break, time=1, value=-1}
        ]}}}
     Scale Curve=50
    )
    Index 2 Param 3=&Waveform(
     Global Input Position=&RandomRangeFloat(Minimum=0, Maximum=15)
     Global Phase=$Particles.Age
     "Waveform [1]"=::NewEnumerator12
     "[1] Phase Scale"=2
     "[1] Frequency"=1
     "[1] Amplitude Min/Max"=<0 1>
    )
    Index 2 Param 4=25
    Index 3 Param 1=&FloatFromCurve(
     FloatCurve=DataInterface() {propertyValues {Curve {keys [
         {interpMode=RCIM_Cubic, leaveTangent=-0.0025, tangentMode=RCTM_Break, time=0, value=0}
         {arriveTangent=1.8528, interpMode=RCIM_Cubic, tangentMode=RCTM_Break, time=1, value=1}
        ]}}}
     Scale Curve=&RandomRangeFloat(Minimum=15, Maximum=25)
    )
    Index 3 Param 2=&FloatFromCurve(FloatCurve=DataInterface() {propertyValues {Curve {keys [
         {interpMode=RCIM_Cubic, leaveTangent=-0.0025, tangentMode=RCTM_Break, time=0, value=0.5}
         {arriveTangent=-0.0316, interpMode=RCIM_Cubic, tangentMode=RCTM_Break, time=1, value=-0.2}
        ]}}})
    Index 3 Param 3=&FloatFromCurve(FloatCurve=DataInterface() {propertyValues {Curve {keys [
         {interpMode=RCIM_Cubic, leaveTangent=1.1034, tangentMode=RCTM_Break, time=0, value=-1}
         {arriveTangent=2.2399, tangentMode=RCTM_User, time=0.5, value=0.125}
        ]}}})
    Index 3 Param 4=&FloatFromCurve(FloatCurve=DataInterface() {propertyValues {Curve {keys [
         {interpMode=RCIM_Cubic, leaveTangent=3.0601, tangentMode=RCTM_Break, time=0.5, value=-0.5}
         {arriveTangent=3.2053, tangentMode=RCTM_User, time=1, value=1}
        ]}}})
    Module.Index0Param0Write=true
    Module.Index0Param1Write=true
    Module.Index0Param2Write=true
    Module.Index0Param3Write=true
    Module.Index1Param0Write=true
    Module.Index1Param1Write=true
    Module.Index1Param2Write=true
    Module.Index1Param3Write=true
    Module.Index2Param0Write=true
    Module.Index2Param1Write=true
    Module.Index2Param2Write=true
    Module.Index2Param3Write=true
    Module.Index3Param0Write=true
    Module.Index3Param1Write=true
    Module.Index3Param2Write=true
    Module.Index3Param3Write=true
    Write Parameter Index 1=true
    Write Parameter Index 2=true
    Write Parameter Index 3=true
   )
   &SolveForcesAndVelocity(Clamp Velocity=false, Limit Acceleration=false)
   &ScaleColor(
    Scale Alpha=&FloatFromCurve(FloatCurve=DataInterface() {propertyValues {Curve {keys [
         {tangentMode=RCTM_User, time=0, value=1}
         {arriveTangent=-0.0405, interpMode=RCIM_Cubic, leaveTangent=-0.0405, tangentMode=RCTM_User, time=0.7009, value=1}
         {arriveTangent=-8.0089, interpMode=RCIM_Cubic, leaveTangent=-8.0089, tangentMode=RCTM_User, time=1, value=0}
        ]}}})
    ScaleA=true
    ScaleRGB=false
   )
  ]
  rendererClasses [&NiagaraMeshRendererProperties, &NiagaraSpriteRendererProperties]
  renderers [
   {rendererClass=&NiagaraMeshRendererProperties, rendererIndex=0}
   {rendererClass=&NiagaraMeshRendererProperties, rendererIndex=1}
   {rendererClass=&NiagaraSpriteRendererProperties, rendererIndex=2}
  ]
 }
 LightningBig {
  emitterUpdateScript [
   &EmitterState()
   &SpawnBurst_Instantaneous(
    Spawn Count=&MakeCustomIntFromBool(Bool=&SetBoolByFloatComparison(A=$User.Spread, B=0.5), True INT=1, False INT=0)
    Spawn Time=0.02
    Use Spawn Probability=false
    UseLoopCountLimit=false
   )
   &SpawnBurst_Instantaneous(
    Spawn Count=&MakeCustomIntFromBool(Bool=&SetBoolByFloatComparison(A=$User.Spread, B=0.5), True INT=1, False INT=0)
    Spawn Group=1
    Spawn Time=0.31
    Use Spawn Probability=false
    UseLoopCountLimit=false
   ) {module=SpawnBurst_Instantaneous001}
  ]
  particleSpawnScript [
   &InitializeParticle(
    Color Maximum=<rgba 2 0 0.8642 1>
    Color Minimum=<rgba 1 0.0595 0.921 1>
    Color Mode=::NewEnumerator2
    Lifetime=&RandomRangeFloat(Minimum=0.25, Maximum=0.4)
    Mesh Renderer Array Visibility Mode=::NewEnumerator3
    Mesh Renderer Info=DataInterface() {propertyValues {MeshRenderer=&self:LightningBig_0.NiagaraMeshRendererProperties_1}}
    Mesh Scale Max=<20 20 20>
    Mesh Scale Min=<8 15 15>
    Mesh Scale Mode=::NewEnumerator2
    Position Offset=&MaskVectorBySpawnGroup(
     Masked Vector=&self:Vector4to3(Vector 4:Vector4f=$System.Light_01)
     Passthrough Vector=&self:Vector4to3(Vector 4:Vector4f=$System.Light_02)
    )
    UsePositionOffset=true
    Write Material Random=true
   ) {module=InitializeParticle001}
   &DynamicMaterialParameters(
    Index 0 Param 1=1
    Index 0 Param 2=0.15
    Index 0 Param 3=1
    Index 0 Param 4=1
    Module.Param0WriteEnabled=true
    Module.Param1WriteEnabled=true
    Module.Param2WriteEnabled=true
    Module.Param3WriteEnabled=true
   )
   &AddVelocity(
    Module.Use Falloff Distance=false
    Origin Offset=<-15000 0 -8000>
    Origin Offset Coordinate Space=::World
    Override Seed=false
    Velocity Mode=::NewEnumerator1
    Velocity Origin=&SubtractVectorFromPosition(
     Position=$Engine.Owner.Position
     Vector=&MakeVector(X=0, Y=0, Z=&Multiply_Float(A=$System.MaxHeight, B=0.9))
    )
    Velocity Speed=5
   ) {module=AddVelocity001}
   &InitialMeshOrientation(
    Mesh Orientation Mode=::NewEnumerator2
    Orientation Vector=<-1 0 0>
    Rotation=&RandomRangeVector(Minimum=<0 0 0>, Maximum=<1 0 0>)
    Use Rotation Vector=true
   )
   &CameraOffset(Camera Offset Amount=-500)
  ]
  particleUpdateScript [
   &SolveForcesAndVelocity(Clamp Velocity=false, Limit Acceleration=false)
   &ParticleState()
   &ScaleMeshSize(Scale Factor=&VectorFromCurve(VectorCurve=DataInterface() {propertyValues {
       XCurve {keys [
         {arriveTangent=0.8627, interpMode=RCIM_Cubic, leaveTangent=0.8627, tangentMode=RCTM_User, time=0, value=0.4}
         {arriveTangent=0.2315, tangentMode=RCTM_User, time=1, value=1}
        ]}
       YCurve {keys [
         {arriveTangent=0.5439, interpMode=RCIM_Cubic, leaveTangent=0.5439, tangentMode=RCTM_User, time=0, value=0.7}
         {arriveTangent=0.2315, tangentMode=RCTM_User, time=1, value=1}
        ]}
       ZCurve {keys [
         {arriveTangent=0.5439, interpMode=RCIM_Cubic, leaveTangent=0.5439, tangentMode=RCTM_User, time=0, value=0.7}
         {arriveTangent=0.2315, tangentMode=RCTM_User, time=1, value=1}
        ]}
      }})) {enabled=false, module=ScaleMeshSize002}
   &DynamicMaterialParameters(
    Index 0 Param 1=&FloatFromCurve(FloatCurve=DataInterface() {propertyValues {Curve {keys [
         {arriveTangent=-0.2542, interpMode=RCIM_Cubic, leaveTangent=-0.2542, tangentMode=RCTM_User, time=0, value=1}
         {arriveTangent=-1.6922, interpMode=RCIM_Cubic, leaveTangent=-1.6922, tangentMode=RCTM_User, time=0.4987, value=0.5013}
         {arriveTangent=-0.9805, interpMode=RCIM_Cubic, leaveTangent=-0.9805, tangentMode=RCTM_User, time=1, value=0}
        ]}}})
    Index 0 Param 2=&FloatFromCurve(
     FloatCurve=DataInterface() {propertyValues {Curve {keys [
         {interpMode=RCIM_Cubic, leaveTangent=-1.1405, tangentMode=RCTM_Break, time=0, value=0.5}
         {arriveTangent=4.4936, interpMode=RCIM_Cubic, leaveTangent=-5.3647, tangentMode=RCTM_Break, time=0.4833, value=1}
         <at 1 0>
        ]}}}
     Scale Curve=0
    )
    Index 0 Param 3=&FloatFromCurve(
     FloatCurve=DataInterface() {propertyValues {Curve {keys [
         <at 0 1 RCIM_Cubic>
         {arriveTangent=-3.8306, interpMode=RCIM_Cubic, tangentMode=RCTM_Break, time=0.49, value=0.1}
         {arriveTangent=-4.2365, interpMode=RCIM_Cubic, leaveTangent=-4.2365, tangentMode=RCTM_User, time=0.5, value=1}
         <at 1 0.2 RCIM_Cubic>
        ]}}}
     Scale Curve=0.5
    )
    Index 0 Param 4=&RandomRangeFloat(Minimum=0, Maximum=1)
    Index 3 Param 1=3
    Index 3 Param 2=1000
    Module.Index0Float0Write=true
    Module.Index0Float1Write=true
    Module.Index0Float2Write=true
    Module.Index0Float3Write=true
    Module.Index3Float0Write=true
    Module.Index3Float1Write=true
    Module.Index3Float2Write=true
    Module.Index3Float3Write=true
    Write Parameter Index 3=::NewEnumerator1
   ) {module=DynamicMaterialParameters001}
   &ScaleColor(
    Scale Alpha=&FloatFromCurve(FloatCurve=DataInterface() {propertyValues {Curve {keys [
         <at 0 0>
         {interpMode=RCIM_Cubic, tangentMode=RCTM_User, time=0.1258, value=0.1952}
         {arriveTangent=0.1568, interpMode=RCIM_Cubic, leaveTangent=0.1568, time=0.341, value=0.0902}
         {interpMode=RCIM_Cubic, tangentMode=RCTM_User, time=0.496, value=0.2532}
         <at 1 0>
        ]}}})
    Scale RGB=&VectorFromFloat(Value=&FloatFromCurve(FloatCurve=DataInterface() {propertyValues {Curve {keys [
          <at 0 0.5>
          {arriveTangent=7, interpMode=RCIM_Cubic, leaveTangent=7, time=0.45, value=1}
          {arriveTangent=-4.524, interpMode=RCIM_Cubic, leaveTangent=-4.524, tangentMode=RCTM_User, time=0.5, value=4}
          <at 1 1>
         ]}}}))
    ScaleA=true
    ScaleRGB=true
   )
  ]
  rendererClasses [&NiagaraMeshRendererProperties]
  renderers [{rendererClass=&NiagaraMeshRendererProperties, rendererIndex=0}]
 }
 MainSmoke {
  emitterUpdateScript [
   &EmitterState()
   &SpawnBurst_Instantaneous(Spawn Count=60, Use Spawn Probability=false, UseLoopCountLimit=true)
  ]
  particleSpawnScript [&InitializeParticle(
    Color=<rgba 0.5 0.0369 0.2291 1>
    Mass Mode=::NewEnumerator3
    Sprite Size Mode=::NewEnumerator3
    Uniform Sprite Size=&Multiply_Float(A=$System.SmokeCloudSize, B=&RandomRangeFloat(Minimum=0.8, Maximum=1))
    UsePositionOffset=false
    Write Material Random=true
   )]
  particleUpdateScript [
   &ParticleState(Kill Particles When Lifetime Has Elapsed=false, Let Infinitely Lived Particles Die When Emitter Deactivates=true)
   /SetVariables(Particles.Age_New:NiagaraFloat=&ModuloFloat(
     A=&Add_Float(
      A=&ReturnNormalizedExecIndex()
      B=&Multiply_Float(A=$Particles.Age, B=$System.SmokeCloudSpeed)
     )
     B=1
    )) {isSetParameters=true}
   &InterpolateOverTime(
    Rate Of Change:NiagaraFloat=&Lerp_Float(A=3, B=10, Alpha=$Particles.Age_New)
    Target Vector Value:Vector3f=&Lerp_Vector(
     A=<0 0 1>
     B=$Engine.Owner.SystemZAxis
     Alpha=&FloatFromCurve(
      FloatCurve=DataInterface() {propertyValues {Curve {
         defaultValue=1
         keys [
          <at 0.1 0 RCIM_Cubic>
          {arriveTangent=2.424, interpMode=RCIM_Cubic, leaveTangent=2.424, tangentMode=RCTM_User, time=0.7277, value=0.6373}
          {tangentMode=RCTM_User, time=0.95, value=1}
         ]
        }}}
      CurveIndex=$Particles.Age_New
     )
    )
    Variable Type=::NewEnumerator2
   ) {module=InterpolateOverTime001}
   &InterpolateOverTime(
    Rate Of Change:NiagaraFloat=&Lerp_Float(
     A=2
     B=10
     Alpha=&NormalizeDistanceRange(
      Start Position:NiagaraPosition=$Engine.Owner.Position
      End Position:NiagaraPosition=$Particles.Position
      Distance:NiagaraFloat=18000
      Invert Normalized Range:NiagaraBool=true
     )
    )
    Target Position Value:NiagaraPosition=&Multiply_Vector(A=Linked() {linkedVariable {name=Engine.Owner.Position, type=NiagaraPosition}}, B=<1 1 0>)
    Variable Type=::NewEnumerator6
   )
   &ShapeLocation(
    Large Radius=&Lerp_Float(
     A=$System.MinRadius
     B=$System.MaxRadius
     Alpha=&FloatFromCurve(
      FloatCurve=DataInterface() {propertyValues {Curve {
         defaultValue=1
         keys [
          {arriveTangent=-0.8855, interpMode=RCIM_Cubic, leaveTangent=-0.8855, leaveTangentWeight=0.0833, tangentMode=RCTM_User, time=0, value=1}
          {
           arriveTangent=-0.9799
           arriveTangentWeight=0.2165
           interpMode=RCIM_Cubic
           leaveTangent=-0.9799
           leaveTangentWeight=0.433
           tangentMode=RCTM_User
           time=0.25
           value=0.75
          }
          {
           arriveTangent=-2.4815
           arriveTangentWeight=0.33
           interpMode=RCIM_Cubic
           leaveTangent=-2.4815
           leaveTangentWeight=0.165
           tangentMode=RCTM_User
           time=0.75
           value=0.25
          }
          {arriveTangent=-1.8365, arriveTangentWeight=0.235, interpMode=RCIM_Cubic, leaveTangent=-1.8365, tangentMode=RCTM_User, time=1, value=0}
         ]
        }}}
      CurveIndex=$Particles.Age_New
      Scale Curve=$User.Spread
     )
    )
    MaskSpawnGroup=false
    Offset=&Multiply_VectorByFloat(
     Vector=&AddVectorToPosition(
      Position=&Multiply_Vector(A=Linked() {linkedVariable {name=Engine.Owner.Position, type=NiagaraPosition}}, B=<0 0 1>)
      Vector=<0 0 3000>
     )
     Float=&Power(a=$Particles.Age_New, b=1.75)
    )
    Offset Coordinate Space=::World
    Offset Mode=::NewEnumerator0
    Rotation Axis=$Particles.InterpolateOverTime001.InterpolateOverTime_V3_.Moving Average
    Rotation Coordinate Space=::World
    Rotation Mode=::NewEnumerator1
    Shape Origin=$Particles.InterpolateOverTime.InterpolateOverTime_Position.Moving Average
    Shape Primitive=::NewEnumerator3
    Torus Distribution Mode:ENiagaraTorusDistributionMode=::NewEnumerator1
    U Position=&Lerp_Float(A=0, B=7.2, Alpha=$Particles.Age_New)
    UseOwnerScale=true
    V Position=&RandomRangeFloat(Minimum=0, Maximum=1)
   )
   /SetVariables(Particles.CameraDot:NiagaraFloat=&RemapRange(
     Input Value To Remap=&DotProduct(
      Vector A=$System.Camera Facing
      Vector B=&NormalizeVector(
       Vector To Normalize=&Multiply_Vector(
        A=&SubtractVectorFromPosition(Position=$Engine.Owner.Position, Vector=Linked() {linkedVariable {name=Particles.Position, type=NiagaraPosition}})
        B=<1 1 0>
       )
       Safe Normalize=true
       Safe Fallback Threshold=0
       Safe Fallback Vector=<0 0 0.001>
      )
     )
     Input Min=-0.33
     Input Max=0.15
     Output Min=0
     Output Max=1
    )) {isSetParameters=true}
   &DynamicMaterialParameters(
    Index 0 Vector 4:Vector4f=$System.Light_01
    Index 1 Vector 4:Vector4f=$System.Light_02
    Index 2 Param 1=&RandomRangeFloat(Minimum=0.9, Maximum=1.05)
    Index 2 Param 4=1.25
    Module.Index2Float0Write=true
    Module.Index2Float1Write=false
    Module.Index2Float2Write=false
    Module.Index2Float3Write=false
    Write Parameter Index 0=::NewEnumerator4
    Write Parameter Index 1=::NewEnumerator4
    Write Parameter Index 2=::NewEnumerator1
   )
   &ScaleColor(
    Scale Alpha=&FloatFromCurve(
     FloatCurve=DataInterface() {propertyValues {Curve {defaultValue=1.0113, keys [<at 0 0>, <at 0.1 1>, {arriveTangent=-1.25, leaveTangent=-1.25, time=0.75, value=1}]}}}
     CurveIndex=$Particles.Age_New
     Scale Curve=$Particles.CameraDot
    )
    Scale RGB=&Multiply_VectorByFloat(Vector=<1 1 1>, Float=1)
    ScaleA=true
    ScaleRGB=true
   )
   &ScaleColor(
    Scale Alpha=&FloatFromCurve(
     FloatCurve=DataInterface() {propertyValues {Curve {
        defaultValue=1
        keys [
         {arriveTangent=4.4995, interpMode=RCIM_Cubic, time=0, value=0}
         {arriveTangent=0.011, interpMode=RCIM_Cubic, leaveTangent=0.011, tangentMode=RCTM_User, time=0.1, value=1}
         <at 1 1 RCIM_Cubic>
        ]
       }}}
     CurveIndex=$System.Timeline.NormalizedTime
    )
    Scale RGB=&Multiply_VectorByFloat(Vector=<1 1 1>, Float=1)
    ScaleA=true
    ScaleRGB=true
   ) {module=ScaleColor001}
   &ScaleSpriteSize(
    Uniform Curve Index=$Particles.Age_New
    Uniform Curve Scale=&Lerp_Float(A=0, B=1, Alpha=$Particles.CameraDot)
    Uniform Curve Sprite Scale=DataInterface() {propertyValues {Curve {
       defaultValue=0.75
       keys [
        {arriveTangent=1.792, interpMode=RCIM_Cubic, leaveTangent=1.792, tangentMode=RCTM_User, time=0.001, value=0.75}
        {arriveTangent=0.0501, interpMode=RCIM_Cubic, leaveTangent=0.0501, time=0.2, value=1}
        {arriveTangent=-0.4017, tangentMode=RCTM_User, time=1, value=0.8}
       ]
      }}}
   )
   &ScaleSpriteSize(
    Uniform Curve Index=&ModuloFloat(
     A=&Add_Float(
      A=&Multiply_Float(
       A=&OneMinusFloat(Float=&ReturnNormalizedExecIndex(Normalized Index Includes One=false))
       B=1.33
      )
      B=&Multiply_Float(A=$Particles.Age, B=0.5)
     )
     B=1
    )
    Uniform Curve Sprite Scale=DataInterface() {propertyValues {Curve {keys [<at 0 1 RCIM_Cubic>,<at 0.3 1.2 RCIM_Cubic>,<at 1 1 RCIM_Cubic>]}}}
   ) {module=ScaleSpriteSize001}
   &CameraOffset(Camera Offset Amount=&Multiply_Float(
     A=/Niagara/DynamicInputs/Math/CalculateParticleRadius(
      Radius Calculation Type:ENiagaraCollisionRadiusOptions=::NewEnumerator0
      Particle Radius:NiagaraFloat=1
      Mesh Dimensions:Vector3f=<1 1 1>
      Method for Calculating Particle Radius:ENiagaraCalculateRadiusOptions=::NewEnumerator2
     )
     B=0.15
    ))
  ]
  rendererClasses [&NiagaraSpriteRendererProperties, &NiagaraMeshRendererProperties]
  renderers [
   {rendererClass=&NiagaraSpriteRendererProperties}
   {rendererClass=&NiagaraMeshRendererProperties, rendererIndex=1}
  ]
 }
}
summary {
 systemName=NS_LimeRock_SmokeMonsterMain
 emitters [
  {emitterName=MainSmoke, rendererClasses [&NiagaraSpriteRendererProperties, &NiagaraMeshRendererProperties]}
  {emitterName=Head, rendererClasses [&NiagaraSpriteRendererProperties, &NiagaraMeshRendererProperties]}
  {emitterName=Beams, rendererClasses [&NiagaraMeshRendererProperties]}
  {emitterName=Lightning, rendererClasses [&NiagaraMeshRendererProperties, &NiagaraSpriteRendererProperties]}
  {emitterName=LightningBig, rendererClasses [&NiagaraMeshRendererProperties]}
 ]
 userVariables [
  User.HeartGlow {type=NiagaraFloat, defaultValue=0, description=""}
  User.Spread {type=NiagaraFloat, defaultValue=0, description=""}
 ]
}
