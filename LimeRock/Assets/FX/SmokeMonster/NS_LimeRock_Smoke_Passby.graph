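# Niagara smoke pass-by effect in the SmokeMonster FX set, continuously emitting sprites that drift along X with drag.
# SmokeSpread biases spawns across local Y away from the centre, where CenterFade shortens lifetime, slows particles and lowers a material parameter.
# Sprite spin follows the camera-to-velocity dot product, and a radius-based camera offset shifts each puff.
# Disabled SmokePours builds three curve-shaped ribbon arcs positioned by System.PortalRadius.
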NiagaraSystem /LimeRock/Assets/FX/SmokeMonster/NS_LimeRock_Smoke_Passby
dependencies {
 usedDataInterfaces [
  &NiagaraDataInterfaceArrayFloat3
  &NiagaraDataInterfaceArrayFloat
  &NiagaraDataInterfaceCurve
  &NiagaraDataInterfaceCamera
 ]
 usedDynamicInputs []
 usedModules [
  &SystemState
  &self:SystemSpawnScript.SystemScriptSource.SystemScriptGraph.NiagaraNodeAssignment_1.SetVariables_#1
 ]
}
emitters {
 SmokePours {
  bEnabled=false
  emitterUpdateScript [
   &EmitterState()
   &SpawnBurst_Instantaneous(Spawn Count=8, Use Spawn Probability=false, UseLoopCountLimit=false)
   &SpawnBurst_Instantaneous(Spawn Count=8, Spawn Time=0.5, Use Spawn Probability=false, UseLoopCountLimit=false) {module=SpawnBurst_Instantaneous001}
   &SpawnBurst_Instantaneous(Spawn Count=8, Spawn Time=0.7, Use Spawn Probability=false, UseLoopCountLimit=false) {module=SpawnBurst_Instantaneous002}
  ]
  particleSpawnScript [
   &PartitionParticles(
    Normalized Index In Partition Includes One=true
    Number Of Partitions=3
    Partition Mode:ENiagara_PartitionMode_v2=::NewEnumerator1
    Write Partition Index to Ribbon Link Order:ENiagara_AttributeSamplingApplyOutput=::NewEnumerator0
    Write Partition to Ribbon ID=::NewEnumerator0
   )
   &InitializeParticle(
    Color Mode=::NewEnumerator0
    Lifetime=6
    Mesh Scale=&SelectVectorFromArray(
     Direct Array Index=$Particles.PartitionParticles.Partition
     Vector Selection Array=DataInterface() {propertyValues {FloatData [<40 20 18>, <14 15 15>, <17 15 12>]}}
    )
    Mesh Scale Mode=::NewEnumerator4
    Ribbon Width=&SelectFloatFromArray(
     Direct Array Index=$Particles.PartitionParticles.Partition
     Float Selection Array=DataInterface() {propertyValues {FloatData [6000,3000,4500]}}
    )
    Ribbon Width Mode=::NewEnumerator1
    UsePositionOffset=false
    Write Material Random=true
   )
   &InitialMeshOrientation(
    Mesh Orientation Mode=::NewEnumerator2
    Orientation Coordinate Space=::World
    Orientation Vector=&Multiply_Vector(A=$Engine.Owner.SystemZAxis, B=<1 1 0>)
    Use Rotation Vector=false
   )
   &DynamicMaterialParameters(
    Index 0 Param 1=-1
    Index 0 Param 2=&Multiply_FloatByInt(Float=0.625, Integer=$Particles.PartitionParticles.Partition)
    Index 0 Param 3=1
    Index 0 Param 4=0.7
    Module.Index0Float0Write=true
    Module.Index0Float1Write=true
    Module.Index0Float2Write=true
    Module.Index0Float3Write=true
   )
  ]
  particleUpdateScript [
   &ParticleState(Kill Particles When Lifetime Has Elapsed=false)
   &ScaleMeshSize(Scale Factor=&MakeVector(
     X=&FloatFromCurve(FloatCurve=DataInterface() {propertyValues {Curve {keys [<at 500 0.25 RCIM_Cubic>,<at 7500 1>]}}}, CurveIndex=$System.PortalRadius)
     Y=1
     Z=&FloatFromCurve(FloatCurve=DataInterface() {propertyValues {Curve {defaultValue=1, keys [{tangentMode=RCTM_User, time=0, value=0.5}, {tangentMode=RCTM_User, time=0.5, value=1}]}}})
    ))
   /SetVariables(Particles.StartPos:NiagaraPosition=&TransformPosition(
     Position=&RotateVectorByQuaternion(
      Vector To Rotate=&MakeVector(X=&Add_Float(A=$System.PortalRadius, B=-200), Y=0, Z=0)
      Quaternion=&MakeQuaternion(
       Quaternion From=0
       Axis=<0 0 1>
       Angle=&Multiply_Float(
        A=&SelectFloatFromArray(
         Direct Array Index=$Particles.PartitionParticles.Partition
         Float Selection Array=DataInterface() {propertyValues {FloatData [0,0.14,-0.12]}}
        )
        B=&FloatFromCurve(FloatCurve=DataInterface() {propertyValues {Curve {keys [<at 500 1 RCIM_Cubic>,<at 7500 0.5>]}}}, CurveIndex=$System.PortalRadius)
       )
       From=?
       To=?
       Angle Type=::NewEnumerator1
       Normalize Vectors=false
      )
     )
     Source Space=::Local
     Destination Space=::Simulation
    )) {isSetParameters=true}
   /SetVariables(Particles.EndPos:NiagaraPosition=&MakeVector(
     X=&MakeFloatFromVector(Vector=Linked() {linkedVariable {name=Particles.StartPos, type=NiagaraPosition}}, Channel=::NewEnumerator0)
     Y=&MakeFloatFromVector(Vector=Linked() {linkedVariable {name=Particles.StartPos, type=NiagaraPosition}}, Channel=::NewEnumerator1)
     Z=&Lerp_Float(
      A=&MakeFloatFromVector(Vector=Linked() {linkedVariable {name=Particles.StartPos, type=NiagaraPosition}}, Channel=::NewEnumerator2)
      B=-1000
      Alpha=&FloatFromCurve(FloatCurve=DataInterface() {propertyValues {Curve {
          defaultValue=1
          keys [{arriveTangent=3.294, interpMode=RCIM_Cubic, time=0, value=0}, {arriveTangent=1.7884, tangentMode=RCTM_User, time=0.4, value=1}]
         }}})
     )
    )) {isSetParameters=true}
   /SetVariables(Particles.Position=&AddVectorToPosition(
     Position=&LerpPosition(
      A=$Particles.StartPos
      B=$Particles.EndPos
      Alpha=&FloatFromCurve(
       FloatCurve=DataInterface() {propertyValues {Curve {defaultValue=1, keys [<at 0 0 RCIM_Cubic>, {arriveTangent=1.4307, tangentMode=RCTM_User, time=1, value=1}]}}}
       CurveIndex=$Particles.PartitionParticles.Normalized Index In Partition
      )
     )
     Vector=&Multiply_VectorByFloat(
      Vector=$Engine.Owner.SystemZAxis
      Float=&FloatFromCurve(
       FloatCurve=DataInterface() {propertyValues {Curve {keys [
           {arriveTangent=15.7741, interpMode=RCIM_Cubic, leaveTangent=15.7741, tangentMode=RCTM_User, time=0, value=-1}
           {arriveTangent=0.7815, interpMode=RCIM_Cubic, leaveTangent=0.7815, tangentMode=RCTM_User, time=0.3, value=0.9}
           <at 0.6 1>
          ]}}}
       CurveIndex=$Particles.PartitionParticles.Normalized Index In Partition
       Scale Curve=1600
      )
     )
    )) {isSetParameters=true}
   &ScaleRibbonWidth(Ribbon Width Scale=&FloatFromCurve(
     FloatCurve=DataInterface() {propertyValues {Curve {defaultValue=1, keys [<at 0 0 RCIM_Cubic>, {tangentMode=RCTM_User, time=0.6, value=1}]}}}
     Scale Curve=&FloatFromCurve(
      FloatCurve=DataInterface() {propertyValues {Curve {keys [{arriveTangent=-0.0054, interpMode=RCIM_Cubic, leaveTangent=-0.0054, tangentMode=RCTM_User, time=0, value=3}, <at 1000 1>]}}}
      CurveIndex=&MakeFloatFromVector(Vector=Linked() {linkedVariable {name=Particles.Position, type=NiagaraPosition}}, Channel=::NewEnumerator2)
     )
    ))
   &ScaleColor(
    Scale Alpha=&FloatFromCurve(FloatCurve=DataInterface() {propertyValues {Curve {keys [<at 0.7 1 RCIM_Cubic>,<at 1 0>]}}})
    ScaleA=true
    ScaleRGB=false
   )
  ]
  rendererClasses [&NiagaraRibbonRendererProperties]
  renderers [{rendererClass=&NiagaraRibbonRendererProperties}]
 }
 SmokeSpread {
  emitterUpdateScript [
   &EmitterState()
   &SpawnRate(SpawnRate=25, Use Spawn Probability=false)
  ]
  particleSpawnScript [
   /SetVariables(Particles.Spread:NiagaraFloat=&FloatFromCurve(
     FloatCurve=DataInterface() {propertyValues {Curve {keys [
         {arriveTangent=-0.4608, interpMode=RCIM_Cubic, leaveTangent=-0.4608, tangentMode=RCTM_User, time=0, value=1}
         {arriveTangent=-1.2, interpMode=RCIM_Cubic, leaveTangent=-1.2, time=0.495, value=0.6}
         {arriveTangent=-1.716, interpMode=RCIM_Cubic, leaveTangent=-1.716, tangentMode=RCTM_User, time=0.5, value=0.4}
         {arriveTangent=-0.6907, interpMode=RCIM_Cubic, leaveTangent=-0.6907, tangentMode=RCTM_User, time=1, value=0}
        ]}}}
     CurveIndex=&RandomRangeFloat(Minimum=0, Maximum=1)
    )) {isSetParameters=true}
   /SetVariables(Particles.CenterFade:NiagaraFloat=&FloatFromCurve(
     FloatCurve=DataInterface() {propertyValues {Curve {
        defaultValue=1
        keys [
         {interpMode=RCIM_Cubic, leaveTangentWeight=0.1, tangentMode=RCTM_User, time=0.1, value=1}
         {arriveTangentWeight=0.1, interpMode=RCIM_Cubic, leaveTangentWeight=0.0667, tangentMode=RCTM_User, time=0.4, value=0}
         {arriveTangentWeight=0.0667, interpMode=RCIM_Cubic, leaveTangentWeight=0.1, tangentMode=RCTM_User, time=0.6, value=0}
         {tangentMode=RCTM_User, time=0.9, value=1}
        ]
       }}}
     CurveIndex=$Particles.Spread
    )) {isSetParameters=true}
   &InitializeParticle(
    Color Mode=::NewEnumerator0
    Lifetime=&Multiply_Float(
     A=&RandomRangeFloat(Minimum=4, Maximum=5)
     B=&Lerp_Float(A=0.3, B=1, Alpha=$Particles.CenterFade)
    )
    Position Offset=&MakeVector(X=0, Y=&Lerp_Float(A=-1200, B=1200, Alpha=$Particles.Spread), Z=0)
    Position Offset Coordinate Space=::Local
    Sprite Rotation Mode=::NewEnumerator2
    Sprite Size Mode=::NewEnumerator1
    Uniform Sprite Size Max=1400
    Uniform Sprite Size Min=1000
    UsePositionOffset=true
    Write Material Random=true
   )
   &ShapeLocation(MaskSpawnGroup=false, Override Seed=false, Rotation Coordinate Space=::World, UseOwnerScale=true)
   &AddVelocity(
    Rotation Coordinate Space=::World
    Velocity=<1 0 0>
    Velocity Speed Scale=&Multiply_Float(
     A=&RandomRangeFloat(Minimum=700, Maximum=900)
     B=&Lerp_Float(A=0.7, B=1, Alpha=$Particles.CenterFade)
    )
   )
  ]
  particleUpdateScript [
   &ParticleState()
   &DynamicMaterialParameters(
    Index 2 Param 1=&RandomRangeFloat(Minimum=0.7, Maximum=1)
    Index 2 Param 2=&FloatFromCurve(
     FloatCurve=DataInterface() {propertyValues {Curve {keys [<at 0 1 RCIM_Cubic>, {arriveTangent=-0.8202, tangentMode=RCTM_User, time=1, value=0.5}]}}}
     Scale Curve=&Lerp_Float(A=0.5, B=0.8, Alpha=$Particles.CenterFade)
    )
    Module.Index2Float0Write=true
    Module.Index2Float1Write=true
    Module.Index2Float2Write=false
    Module.Index2Float3Write=false
    Write Parameter Index 0=::NewEnumerator0
    Write Parameter Index 2=::NewEnumerator1
   )
   &SpriteRotationRate(Rotation Rate=&Multiply_Float(
     A=&DotProduct(
      Vector A=&NormalizeVector(
       Vector To Normalize=&ReturnCameraVector(Camera Vectors=::NewEnumerator3, Coordinate Space=::Simulation)
       Safe Normalize=true
       Safe Fallback Threshold=0
       Safe Fallback Vector=<0 0 0.001>
      )
      Vector B=&NormalizeVector(Vector To Normalize=$Particles.Velocity, Safe Normalize=true, Safe Fallback Threshold=0, Safe Fallback Vector=<0 0 0.001>)
     )
     B=&FloatFromCurve(
      FloatCurve=DataInterface() {propertyValues {Curve {keys [
          {arriveTangent=3.0812, interpMode=RCIM_Cubic, leaveTangent=3.0812, leaveTangentWeight=0.6072, tangentMode=RCTM_User, time=0, value=0.2}
          {arriveTangent=0.2, interpMode=RCIM_Cubic, leaveTangent=0.2, time=0.25, value=1}
          <at 1 0.4>
         ]}}}
      Scale Curve=-80
     )
    ))
   &ScaleSpriteSize(Uniform Curve Sprite Scale=DataInterface() {propertyValues {Curve {keys [
        {arriveTangent=1.7153, interpMode=RCIM_Cubic, leaveTangent=1.7153, tangentMode=RCTM_User, time=0, value=0.3}
        {arriveTangent=0.2491, tangentMode=RCTM_User, time=1, value=1}
       ]}}})
   &Drag(Drag=0.4, Use Linear Drag=true, Use Rotational Drag=false)
   &SolveForcesAndVelocity(Clamp Velocity=false, Limit Acceleration=false)
   &ScaleColor(
    Scale Alpha=&FloatFromCurve(FloatCurve=DataInterface() {propertyValues {Curve {
        defaultValue=1
        keys [
         {arriveTangent=3.8982, interpMode=RCIM_Cubic, leaveTangent=3.8982, tangentMode=RCTM_User, time=0, value=0}
         {interpMode=RCIM_Cubic, tangentMode=RCTM_User, time=0.25, value=1}
         {arriveTangent=-1.7921, tangentMode=RCTM_User, time=1, value=0}
        ]
       }}})
    ScaleA=true
    ScaleRGB=false
   )
   &CameraOffset(Camera Offset Amount=&Multiply_Float(
     A=/Niagara/DynamicInputs/Math/CalculateParticleRadius(
      Radius Calculation Type:ENiagaraCollisionRadiusOptions=::NewEnumerator0
      Particle Radius:NiagaraFloat=1
      Mesh Dimensions:Vector3f=<1 1 1>
      Method for Calculating Particle Radius:ENiagaraCalculateRadiusOptions=::NewEnumerator2
     )
     B=0.5
    ))
  ]
  rendererClasses [&NiagaraSpriteRendererProperties]
  renderers [{rendererClass=&NiagaraSpriteRendererProperties}]
 }
}
summary {
 systemName=NS_LimeRock_Smoke_Passby
 emitters [
  {bEnabled=false, emitterName=SmokePours, rendererClasses [&NiagaraRibbonRendererProperties]}
  {emitterName=SmokeSpread, rendererClasses [&NiagaraSpriteRendererProperties]}
 ]
 userVariables []
}
