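# Smoke Monster surround effect: large smoke sprites that spread along two user-supplied splines, presumably to encircle a target.
# Two burst spawns form separate spawn groups, each mapped to its own spline with opposite sprite rotation signs.
# Normalized spawn index staggers size, spline distance and alpha fade-in, while curves grow and advance the sprites.
# Particles persist past their lifetime, and a clamped radius-based camera offset shifts each sprite.
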NiagaraSystem /LimeRock/Assets/FX/SmokeMonster/NS_LimeRock_Smoke_Surround
dependencies {
 usedDataInterfaces [&NiagaraDataInterfaceCurve, /Script/Niagara.NiagaraDataInterfaceSpline]
 usedDynamicInputs []
 usedModules [
  &self:SystemSpawnScript.SystemScriptSource.SystemScriptGraph.NiagaraNodeAssignment_0.SetVariables_#1
  &SystemState
 ]
}
emitters {Minimal {
  emitterUpdateScript [
   &EmitterState(Life Cycle Mode=::NewEnumerator1, Loop Behavior=::NewEnumerator1, Loop Duration Mode=::NewEnumerator1, UseLoopDelay=false)
   &SpawnBurst_Instantaneous(Spawn Count=12, Use Spawn Probability=false, UseLoopCountLimit=false)
   &SpawnBurst_Instantaneous(Spawn Count=12, Spawn Group=1, Use Spawn Probability=false, UseLoopCountLimit=false) {module=SpawnBurst_Instantaneous001}
  ]
  particleSpawnScript [
   /SetVariables(Particles.InitialExecIndex:NiagaraFloat=&ReturnNormalizedExecIndex()) {isSetParameters=true}
   &InitializeParticle(
    Color=&MakeLinearColor(
     R:NiagaraFloat=1
     G:NiagaraFloat=1
     B:NiagaraFloat=1
     A:NiagaraFloat=&FloatFromCurve(
      FloatCurve=DataInterface() {propertyValues {Curve {keys [<at 0 1 RCIM_Cubic>, {arriveTangent=-0.2137, tangentMode=RCTM_User, time=1, value=0.75}]}}}
      CurveIndex=&ReturnNormalizedExecIndex()
     )
    )
    Lifetime=4
    Sprite Rotation Angle=&Multiply_Float(
     A=&Lerp_Float(A=0.1, B=0.3, Alpha=$Particles.InitialExecIndex)
     B=&MaskFloatBySpawnGroup(Masked Float=-1, Passthrough Float=1.5)
    )
    Sprite Rotation Mode=::NewEnumerator1
    Sprite Size Mode=::NewEnumerator3
    Uniform Sprite Size=&FloatFromCurve(
     FloatCurve=DataInterface() {propertyValues {Curve {keys [
         {arriveTangent=3.0702, interpMode=RCIM_Cubic, leaveTangent=3.0702, tangentMode=RCTM_User, time=0, value=0.6}
         {arriveTangent=-0.2, interpMode=RCIM_Cubic, leaveTangent=-0.2, time=0.2, value=1}
         {arriveTangent=-0.9722, tangentMode=RCTM_User, time=1, value=0.4}
        ]}}}
     CurveIndex=&ReturnNormalizedExecIndex()
     Scale Curve=7500
    )
    UsePositionOffset=false
    Write Material Random=true
   )
   &DynamicMaterialParameters(
    Index 2 Param 1=&Add_Float(
     A=&Lerp_Float(A=0.7, B=1, Alpha=$Particles.InitialExecIndex)
     B=&RandomRangeFloat(Minimum=-0.1, Maximum=0.1)
    )
    Index 2 Param 2=&FloatFromCurve(FloatCurve=DataInterface() {propertyValues {Curve {keys [<at 0 1>,<at 1 0>]}}})
    Module.Index2Float0Write=true
    Module.Index2Float1Write=false
    Module.Index2Float2Write=false
    Module.Index2Float3Write=false
    Write Parameter Index 0=::NewEnumerator0
    Write Parameter Index 2=::NewEnumerator1
   )
  ]
  particleUpdateScript [
   &ParticleState(Kill Particles When Lifetime Has Elapsed=false)
   &ScaleSpriteSize(Uniform Curve Sprite Scale=DataInterface() {propertyValues {Curve {keys [
        {arriveTangent=1.7493, interpMode=RCIM_Cubic, leaveTangent=1.7493, tangentMode=RCTM_User, time=0, value=0.25}
        {tangentMode=RCTM_User, time=1, value=1}
       ]}}})
   &CameraOffset(Camera Offset Amount=&ClampFloat(
     Float=/Niagara/DynamicInputs/Math/CalculateParticleRadius(
      Radius Calculation Type:ENiagaraCollisionRadiusOptions=::NewEnumerator0
      Particle Radius:NiagaraFloat=1
      Mesh Dimensions:Vector3f=<1 1 1>
      Method for Calculating Particle Radius:ENiagaraCalculateRadiusOptions=::NewEnumerator2
     )
     Max=2000
    ))
   &ScaleColor(
    Scale Alpha=&FloatFromCurve(
     FloatCurve=DataInterface() {propertyValues {Curve {defaultValue=1, keys [{tangentMode=RCTM_User, time=0, value=0}, {tangentMode=RCTM_User, time=1, value=1}]}}}
     CurveIndex=&Subtract_Float(A=$Particles.Age, B=&Multiply_Float(A=$Particles.InitialExecIndex, B=3))
    )
    ScaleA=true
    ScaleRGB=false
   )
   /SetVariables(Particles.Position=&MaskVectorBySpawnGroup(
     Masked Vector=&self:SplinePosition(
      Distance:NiagaraFloat=&FloatFromCurve(
       FloatCurve=DataInterface() {propertyValues {Curve {
          defaultValue=1
          keys [
           {arriveTangent=1.4251, interpMode=RCIM_Cubic, leaveTangent=1.4251, tangentMode=RCTM_User, time=0, value=0.1}
           {arriveTangent=0.4965, tangentMode=RCTM_User, time=1, value=1}
          ]
         }}}
       CurveIndex=$Particles.InitialExecIndex
       Scale Curve=&FloatFromCurve(FloatCurve=DataInterface() {propertyValues {Curve {
           defaultValue=1
           keys [
            {arriveTangent=2.1756, interpMode=RCIM_Cubic, leaveTangent=2.1756, tangentMode=RCTM_User, time=0, value=0}
            {tangentMode=RCTM_User, time=1, value=1}
           ]
          }}})
      )
      New Spline:NiagaraDataInterfaceSpline=$System.Spline1
     )
     Passthrough Vector=&self:SplinePosition(
      Distance:NiagaraFloat=&FloatFromCurve(
       FloatCurve=DataInterface() {propertyValues {Curve {
          defaultValue=1
          keys [
           {arriveTangent=1.002, interpMode=RCIM_Cubic, leaveTangent=1.002, tangentMode=RCTM_User, time=0, value=0.1}
           {arriveTangent=0.2711, tangentMode=RCTM_User, time=1, value=1}
          ]
         }}}
       CurveIndex=$Particles.InitialExecIndex
       Scale Curve=&FloatFromCurve(FloatCurve=DataInterface() {propertyValues {Curve {
           defaultValue=1
           keys [
            {arriveTangent=2.1756, interpMode=RCIM_Cubic, leaveTangent=2.1756, tangentMode=RCTM_User, time=0, value=0}
            {tangentMode=RCTM_User, time=1, value=1}
           ]
          }}})
      )
      New Spline:NiagaraDataInterfaceSpline=$System.Spline2
     )
    )) {isSetParameters=true}
  ]
  rendererClasses [&NiagaraSpriteRendererProperties]
  renderers [{rendererClass=&NiagaraSpriteRendererProperties}]
 }}
summary {
 systemName=NS_LimeRock_Smoke_Surround
 emitters [{emitterName=Minimal, rendererClasses [&NiagaraSpriteRendererProperties]}]
 userVariables [
  User.SplineObj1 {
   type=Object
   description=""
   defaultValue {struct=NiagaraExt_VariableValue_Object, value {object=null, objectClass=&Object}}
  }
  User.SplineObj2 {
   type=Object
   description=""
   defaultValue {struct=NiagaraExt_VariableValue_Object, value {object=null, objectClass=&Object}}
  }
 ]
}
