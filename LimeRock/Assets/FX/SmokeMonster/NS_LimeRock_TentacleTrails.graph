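# Niagara system for the LimeRock smoke monster that leaves trails from four skeletal meshes, presumably its tentacles.
# System-level samplers supply position and velocity for the user-supplied meshes.
# One emitter partitions particles into four ribbon IDs, selecting each partition's spawn position and inherited velocity from its sampler, plus extra velocity.
# Drag slows particles while curves shape ribbon width, alpha fade and dynamic material parameters for the sprite and ribbon renderers.
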NiagaraSystem /LimeRock/Assets/FX/SmokeMonster/NS_LimeRock_TentacleTrails
dependencies {
 usedDataInterfaces [&NiagaraDataInterfaceCurve]
 usedDynamicInputs []
 usedModules [&SystemState, &SampleSkeletalMesh]
}
emitters {Minimal {
  emitterUpdateScript [
   &EmitterState()
   &SpawnRate(SpawnRate=60, Use Spawn Probability=false)
  ]
  particleSpawnScript [
   &PartitionParticles(Number Of Partitions=4, Write Partition to Ribbon ID=::NewEnumerator0)
   &InitializeParticle(
    Color=&MakeLinearColorFromVectorAndFloat(
     "Vector (RGB)"=<1 1 1>
     "Float (Alpha)"=&FloatFromCurve(
      FloatCurve=DataInterface() {propertyValues {Curve {
         defaultValue=1
         keys [
          {arriveTangent=16.9484, interpMode=RCIM_Cubic, leaveTangent=16.9484, tangentMode=RCTM_User, time=0, value=0}
          {interpMode=RCIM_Cubic, tangentMode=RCTM_User, time=0.1, value=1}
          {arriveTangent=-3.2882, tangentMode=RCTM_User, time=1, value=0}
         ]
        }}}
      CurveIndex=$Emitter.NormalizedLoopAge
     )
    )
    Position=&self:SelectPosition(
     Position1:NiagaraPosition=$System.SampleSkeletalMesh.SampledPosition
     Position2:NiagaraPosition=$System.SampleSkeletalMesh001.SampledPosition
     Position3:NiagaraPosition=$System.SampleSkeletalMesh002.SampledPosition
     Position4:NiagaraPosition=$System.SampleSkeletalMesh003.SampledPosition
     Index:NiagaraInt32=$Particles.PartitionParticles.Partition
    )
    Position Mode=::NewEnumerator1
    Ribbon Width=1700
    Ribbon Width Mode=::NewEnumerator1
    Sprite Size Mode=::NewEnumerator3
    Uniform Sprite Size=100
    UsePositionOffset=false
    Write Material Random=true
   )
   &AddVelocity(
    Velocity=&self:SelectVector(
     Vector1:Vector3f=$System.SampleSkeletalMesh.SampledVelocity
     Vector2:Vector3f=$System.SampleSkeletalMesh001.SampledVelocity
     Vector3:Vector3f=$System.SampleSkeletalMesh002.SampledVelocity
     Vector4:Vector3f=$System.SampleSkeletalMesh003.SampledVelocity
     Index:NiagaraInt32=$Particles.PartitionParticles.Partition
    )
    Velocity Speed Scale=0.5
   )
   &AddVelocity(Module.Use Falloff Distance=false, Override Seed=false, Velocity Mode=::NewEnumerator1, Velocity Speed=1500) {module=AddVelocity001}
  ]
  particleUpdateScript [
   &ParticleState()
   &Drag(Drag=1.5, Use Linear Drag=true, Use Rotational Drag=false)
   &SolveForcesAndVelocity(Clamp Velocity=false, Limit Acceleration=false)
   &DynamicMaterialParameters(
    Index 0 Param 1=-5
    Index 0 Param 3=&FloatFromCurve(
     FloatCurve=DataInterface() {propertyValues {Curve {
        defaultValue=1
        keys [
         {arriveTangent=5.2129, interpMode=RCIM_Cubic, leaveTangent=5.2129, tangentMode=RCTM_User, time=0, value=0.5}
         {interpMode=RCIM_Cubic, tangentMode=RCTM_User, time=0.1, value=1}
         {tangentMode=RCTM_User, time=1, value=0.25}
        ]
       }}}
     Scale Curve=0.7
    )
    Index 0 Param 4=&FloatFromCurve(FloatCurve=DataInterface() {propertyValues {Curve {keys [<at 0 0.3 RCIM_Cubic>,<at 1 1>]}}})
    Module.Index0Float0Write=true
    Module.Index0Float1Write=false
    Module.Index0Float2Write=true
    Module.Index0Float3Write=true
   )
   &ScaleColor(
    Scale Alpha=&FloatFromCurve(FloatCurve=DataInterface() {propertyValues {Curve {keys [
         {arriveTangent=-1.8218, interpMode=RCIM_Cubic, leaveTangent=-1.8218, tangentMode=RCTM_User, time=0, value=1}
         {arriveTangent=-0.9707, interpMode=RCIM_Cubic, leaveTangent=-0.9707, tangentMode=RCTM_User, time=1, value=0}
        ]}}})
    ScaleA=true
    ScaleRGB=false
   )
   &ScaleRibbonWidth(Ribbon Width Scale=&FloatFromCurve(FloatCurve=DataInterface() {propertyValues {Curve {
        defaultValue=1
        keys [
         {arriveTangent=7.6201, interpMode=RCIM_Cubic, leaveTangent=7.6201, tangentMode=RCTM_User, time=0, value=0.4}
         {arriveTangent=0.5201, interpMode=RCIM_Cubic, leaveTangent=0.5201, tangentMode=RCTM_User, time=0.2, value=1}
         {arriveTangent=-1.5869, tangentMode=RCTM_User, time=1, value=0.2}
        ]
       }}}))
  ]
  rendererClasses [&NiagaraSpriteRendererProperties, &NiagaraRibbonRendererProperties]
  renderers [
   {rendererClass=&NiagaraSpriteRendererProperties}
   {rendererClass=&NiagaraRibbonRendererProperties, rendererIndex=1}
  ]
 }}
summary {
 systemName=NS_LimeRock_TentacleTrails
 emitters [{emitterName=Minimal, rendererClasses [&NiagaraSpriteRendererProperties, &NiagaraRibbonRendererProperties]}]
 userVariables [
  User.Skeletal Mesh1 {
   type=NiagaraDataInterfaceSkeletalMesh
   defaultValue {
    struct=NiagaraExt_VariableValue_DataInterface
    value {dataInterface=&self:NiagaraDataInterfaceSkeletalMesh_0}
   }
  }
  User.Skeletal Mesh2 {
   type=NiagaraDataInterfaceSkeletalMesh
   defaultValue {
    struct=NiagaraExt_VariableValue_DataInterface
    value {dataInterface=&self:NiagaraDataInterfaceSkeletalMesh_1}
   }
  }
  User.Skeletal Mesh3 {
   type=NiagaraDataInterfaceSkeletalMesh
   defaultValue {
    struct=NiagaraExt_VariableValue_DataInterface
    value {dataInterface=&self:NiagaraDataInterfaceSkeletalMesh_2}
   }
  }
  User.Skeletal Mesh4 {
   type=NiagaraDataInterfaceSkeletalMesh
   defaultValue {
    struct=NiagaraExt_VariableValue_DataInterface
    value {dataInterface=&self:NiagaraDataInterfaceSkeletalMesh_3}
   }
  }
 ]
}
